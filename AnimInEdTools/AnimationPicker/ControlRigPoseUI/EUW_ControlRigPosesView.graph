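# Editor Utility Widget that shows a titled, dimmed popup for picking Control Rig poses in the animation picker tool.
# On Construct it sets the title and creates an EUW_ControlRigPoses grid from the pose path map, Control Rig binding and column count, placing it in a named slot.
# The popup removes itself when a pose thumbnail is clicked, Cancel is pressed, or OnMouseButtonDown fires on it.
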EditorUtilityWidgetBlueprint /AnimInEdTools/AnimationPicker/ControlRigPoseUI/EUW_ControlRigPosesView
graphs [
 EventGraph {nodes [
   n1 = "On Clicked (CancelButton)"() {is=title K2Node_ComponentBoundEvent}
   RemoveFromParent(execute=[@n1.then, @n2.then])
   n2 = Remove From Parent Event() {is=title K2Node_CustomEvent}
   Event Construct()
   > Set Text(Text=Get Title() -> (Title:Text), self=Get TitleTextBlock() -> (TitleTextBlock:Text Object Reference))
   > n3 = Create Widget!(
    execute:Exec
    WorldContextObject:Object Reference
    Class:User Widget Class Reference=&/EUW_ControlRigPoses.EUW_ControlRigPoses_C
    OwningPlayer:Player Controller Object Reference
    GridCollumsNumber:int=Get Grid Collums Number() -> (Grid Collums Number:int=0)
    ThumbnailResolution:int=100
    Control Rig Poses Paths:Map of Names to Names=Get Control Rig Poses Paths() -> (Control Rig Poses Paths:Map of Names to Names)
    Current Control Rig:Control Rig Sequencer Binding Proxy Structure=Get Current Control Rig() -> (Current Control Rig:Control Rig Sequencer Binding Proxy Structure)
   ) -> (then:Exec, ReturnValue:EUW Control Rig Poses Object Reference)
   > Bind Event to Thumbnail Button Clicked Dispatcher(self:EUW Control Rig Poses Object Reference=@n3, Delegate=@n2) {is=title K2Node_AddDelegate}
   > AddChild(self=Get ControlRigPosesWidget_NamedSlot() -> (ControlRigPosesWidget_NamedSlot:Named Slot Object Reference), Content=@n3)
   Event PreConstruct()
   Event Tick()
  ]}
 OnMouseButtonDown_0 {nodes [
   OnMouseButtonDown_0!() -> (then, MyGeometry:Geometry Structure, MouseEvent:"Pointer Event Structure (by ref)") {is=title K2Node_FunctionEntry}
   > RemoveFromParent()
   > Return Node!(execute, ReturnValue:Event Reply Structure=Make Event Reply!() -> (EventReply:Event Reply Structure) {is=title K2Node_MakeStruct})
  ]}
]
variables [
 Control Rig Poses Paths {type="map<string>", default {}}
 Current Control Rig {
  type=ControlRigSequencerBindingProxy
  default {controlRig=null, track=null, proxy {bindingId=#1, sequence=null}}
 }
 Grid Collums Number {type=integer, default=6}
 Title {type=string, default=""}
]
widgets {
 info {rootWidgetClass=&Border, widgetCount=14}
 widgets [
  Border_0 {
   widgetClassPath=&Border
   widget=Border_0
   properties {horizontalAlignment=HAlign_Center, verticalAlignment=VAlign_Center, brushColor=<rgba 0 0 0 0.7>}
  }
  SizeBox_0 {
   widgetClassPath=&SizeBox
   parent=Border_0
   widget=SizeBox_0
   properties {
    bOverride_MinDesiredHeight=true
    bOverride_MinDesiredWidth=true
    maxDesiredHeight=500
    maxDesiredWidth=500
    minDesiredHeight=200
    minDesiredWidth=400
   }
   slot {class=BorderSlot, properties {horizontalAlignment=HAlign_Center, verticalAlignment=VAlign_Center}}
  }
  Border_1 {
   widgetClassPath=&Border
   parent=SizeBox_0
   widget=Border_1
   properties {brushColor=<rgba 0.0157 0.0157 0.0157 1>, padding=<ltrb 5 5 5 5>}
   slot {class=SizeBoxSlot, properties {padding=<ltrb 5 5 5 5>}}
  }
  VerticalBox_0 {widgetClassPath=&VerticalBox, parent=Border_1, widget=VerticalBox_0, slot {class=BorderSlot, properties {padding=<ltrb 5 5 5 5>}}}
  SizeBox_1 {
   widgetClassPath=&SizeBox
   parent=VerticalBox_0
   widget=SizeBox_1
   properties {bOverride_MinDesiredHeight=true, minDesiredHeight=20}
   slot {class=VerticalBoxSlot, properties {verticalAlignment=VAlign_Center, padding=<ltrb 0 0 0 3>}}
  }
  Border_274 {
   widgetClassPath=&Border
   parent=SizeBox_1
   widget=Border_274
   properties {verticalAlignment=VAlign_Center, brushColor=<rgba 0.0275 0.0275 0.0275 1>, padding=<ltrb 5 5 5 5>}
   slot {class=SizeBoxSlot}
  }
  TitleTextBlock {
   widgetClassPath=&TextBlock
   bIsVariable=true
   parent=Border_274
   widget=TitleTextBlock
   properties {
    text=Title
    colorAndOpacity {colorUseRule=UseColor_Specified, specifiedColor=<rgba 0.5255 0.5255 0.5255 1>}
    font {
     bForceMonospaced=false
     bMaterialIsStencil=false
     fontMaterial=null
     fontObject=&Roboto
     letterSpacing=0
     monospacedWidth=1
     size=10
     skewAmount=0
     typefaceFontName=Bold
     outlineSettings {
      bApplyOutlineToDropShadows=false
      bMiteredCorners=false
      bSeparateFillAlpha=false
      outlineMaterial=null
      outlineSize=0
      outlineColor=<rgba 0 0 0 1>
     }
    }
   }
   slot {class=BorderSlot, properties {verticalAlignment=VAlign_Center, padding=<ltrb 5 5 5 5>}}
  }
  Border_344 {
   widgetClassPath=&Border
   parent=VerticalBox_0
   widget=Border_344
   properties {brushColor=<rgba 0.0273 0.0273 0.0273 1>, padding=<ltrb 5 5 5 5>}
   slot {class=VerticalBoxSlot, properties {size {sizeRule=Fill, value=1}}}
  }
  VerticalBox_1 {
   widgetClassPath=&VerticalBox
   parent=Border_344
   widget=VerticalBox_1
   slot {class=BorderSlot, properties {padding=<ltrb 5 5 5 5>}}
  }
  Border_235 {
   widgetClassPath=&Border
   parent=VerticalBox_1
   widget=Border_235
   properties {brushColor=<rgba 1 1 1 0>, padding=<ltrb 0 0 0 0>}
   slot {class=VerticalBoxSlot, properties {size {sizeRule=Fill, value=1}}}
  }
  ControlRigPosesWidget_NamedSlot {
   widgetClassPath=&NamedSlot
   bIsVariable=true
   parent=Border_235
   widget=ControlRigPosesWidget_NamedSlot
   slot {class=BorderSlot, properties {padding=<ltrb 0 0 0 0>}}
  }
  ButtonBox {
   widgetClassPath=&HorizontalBox
   parent=VerticalBox_1
   widget=ButtonBox
   slot {class=VerticalBoxSlot, properties {verticalAlignment=VAlign_Bottom, padding=<ltrb 5 5 5 5>}}
  }
  CancelButton {
   widgetClassPath=&Button
   parent=ButtonBox
   widget=CancelButton
   properties {widgetStyle {
     clickedSlateSound {resourceObject=null}
     disabled {
      bIsDynamicallyLoaded=false
      drawAs=RoundedBox
      imageType=NoImage
      mirroring=NoMirror
      resourceName=None
      resourceObject=null
      tiling=NoTile
      imageSize=<0 0>
      margin=<ltrb 0 0 0 0>
      outlineSettings {
       bUseBrushTransparency=false
       roundingType=FixedRadius
       width=1
       color {colorUseRule=UseColor_ColorTable, specifiedColor=<rgba 0 0 0 0>}
       cornerRadii=<4 4 4 4>
      }
      tintColor {colorUseRule=UseColor_Specified, specifiedColor=<rgba 0.0395 0.0395 0.0395 1>}
      uVRegion {bIsValid=false, max=<0 0>, min=<0 0>}
     }
     disabledForeground {colorUseRule=UseColor_ColorTable, specifiedColor=<rgba 0 0 0 0>}
     hovered {
      bIsDynamicallyLoaded=false
      drawAs=RoundedBox
      imageType=NoImage
      mirroring=NoMirror
      resourceName=None
      resourceObject=null
      tiling=NoTile
      imageSize=<32 32>
      margin=<ltrb 0 0 0 0>
      outlineSettings {
       bUseBrushTransparency=false
       roundingType=FixedRadius
       width=1
       color {colorUseRule=UseColor_ColorTable, specifiedColor=<rgba 0 0 0 0>}
       cornerRadii=<4 4 4 4>
      }
      tintColor {colorUseRule=UseColor_Specified, specifiedColor=<rgba 0.0953 0.0953 0.0953 1>}
      uVRegion {bIsValid=false, max=<0 0>, min=<0 0>}
     }
     hoveredForeground {colorUseRule=UseColor_ColorTable, specifiedColor=<rgba 0 0 0 0>}
     hoveredSlateSound {resourceObject=null}
     normal {
      bIsDynamicallyLoaded=false
      drawAs=RoundedBox
      imageType=NoImage
      mirroring=NoMirror
      resourceName=None
      resourceObject=null
      tiling=NoTile
      imageSize=<32 32>
      margin=<ltrb 0 0 0 0>
      outlineSettings {
       bUseBrushTransparency=false
       roundingType=FixedRadius
       width=1
       color {colorUseRule=UseColor_ColorTable, specifiedColor=<rgba 0 0 0 0>}
       cornerRadii=<4 4 4 4>
      }
      tintColor {colorUseRule=UseColor_Specified, specifiedColor=<rgba 0.3991 0.0294 0.0543 1>}
      uVRegion {bIsValid=false, max=<0 0>, min=<0 0>}
     }
     normalForeground {colorUseRule=UseColor_ColorTable, specifiedColor=<rgba 0 0 0 0>}
     normalPadding=<ltrb 12 1.5 12 1.5>
     pressed {
      bIsDynamicallyLoaded=false
      drawAs=RoundedBox
      imageType=NoImage
      mirroring=NoMirror
      resourceName=None
      resourceObject=null
      tiling=NoTile
      imageSize=<32 32>
      margin=<ltrb 0 0 0 0>
      outlineSettings {
       bUseBrushTransparency=false
       roundingType=FixedRadius
       width=1
       color {colorUseRule=UseColor_ColorTable, specifiedColor=<rgba 0 0 0 0>}
       cornerRadii=<4 4 4 4>
      }
      tintColor {colorUseRule=UseColor_Specified, specifiedColor=<rgba 0.0284 0.0284 0.0284 1>}
      uVRegion {bIsValid=false, max=<0 0>, min=<0 0>}
     }
     pressedForeground {colorUseRule=UseColor_ColorTable, specifiedColor=<rgba 0 0 0 0>}
     pressedPadding=<ltrb 12 2.5 12 0.5>
     pressedSlateSound {resourceObject=null}
    }}
   slot {class=HorizontalBoxSlot, properties {size {sizeRule=Fill, value=1}}}
  }
  CancelTextBlock {
   widgetClassPath=&TextBlock
   bIsVariable=true
   parent=CancelButton
   widget=CancelTextBlock
   properties {
    autoWrapText=true
    justification=Center
    text=Cancel
    colorAndOpacity {colorUseRule=UseColor_Specified, specifiedColor=<rgba 0.5255 0.5255 0.5255 1>}
    font {
     bForceMonospaced=false
     bMaterialIsStencil=false
     fontMaterial=null
     fontObject=&Roboto
     letterSpacing=0
     monospacedWidth=1
     size=10
     skewAmount=0
     typefaceFontName=Bold
     outlineSettings {
      bApplyOutlineToDropShadows=false
      bMiteredCorners=false
      bSeparateFillAlpha=false
      outlineMaterial=null
      outlineSize=0
      outlineColor=<rgba 0 0 0 1>
     }
    }
   }
   slot {class=ButtonSlot}
  }
 ]
}
