EditorUtilityWidgetBlueprint /AnimInEdTools/LinkedSequenceManager/EUW_LinkedSequenceManger
graphs [
 Anim Dialog {nodes [
   n1 = Anim Dialog!() -> (then, AnimSequence Path:String, LevelSequence Path:String, Animatable Creature:Animatable Creature Object Reference) {is=title K2Node_FunctionEntry}
   > Branch(Condition=IsValid!(InBinding:"Movie Scene Binding Proxy Structure (by ref)"=Get Current Binding() -> (Current Binding:Movie Scene Binding Proxy Structure)) -> (ReturnValue))
   > n2 = Create Anim Sequence!(
    execute
    self
    Level Sequence:Level Sequence Object Reference=Get Level Sequence() -> (Level Sequence:Level Sequence Object Reference)
    New AnimSequence Path:String=Get AnimSequence Path() -> (AnimSequence Path:String)
    Binding:Movie Scene Binding Proxy Structure=Get Current Binding()
    Project:BP Animation Browser Project Object Reference="Get (a copy)"(Array:Array of BP Animation Browser Project Object References=Get Projects!(self:Animatable Creature Object Reference=@n1.Animatable Creature) -> (Projects:Array of BP Animation Browser Project Object References) {is=title K2Node_CallFunction}) -> (Output:BP Animation Browser Project Object Reference)
   ) -> (then, Anim Sequence:Anim Sequence Object Reference)
   > n3 = For Each Loop(Array:Array of EUW Linked Sequence Row Object References=Get EUW Rows() -> (EUW Rows:Array of EUW Linked Sequence Row Object References)) -> (Array Element:EUW Linked Sequence Row Object Reference)
   > n4 = GetName!(
    execute
    InBinding:"Movie Scene Binding Proxy Structure (by ref)"=Get In Binding!(self:EUW Linked Sequence Row Object Reference=@n3) -> (In Binding:Movie Scene Binding Proxy Structure)
   ) -> (then, ReturnValue:String)
   > n5 = GetName(InBinding=Get Current Binding())
   > Branch(Condition="Equal Exactly (String)"(A=@n5, B=@n4))
   > Set AnimSequence!(execute, AnimSequence:Anim Sequence Object Reference=@n2, self:EUW Linked Sequence Row Object Reference=@n3) -> (then, Output_Get:Anim Sequence Object Reference)
   > Update Appearance(self:EUW Linked Sequence Row Object Reference=@n3)
  ]}
 Create Anim Dialog {nodes [
   n1 = Create Anim Dialog!() -> (then, Binding:Movie Scene Binding Proxy Structure) {is=title K2Node_FunctionEntry}
   > Set Current Binding(Current Binding:Movie Scene Binding Proxy Structure=@n1) -> (Output_Get:Movie Scene Binding Proxy Structure)
   > n2 = Construct Object from Class(Class=/AnimInEdTools/AnimationBrowser/EUW_Animation_Dialog.EUW_Animation_Dialog_C) -> (ReturnValue:EUW Animation Dialog Object Reference)
   > AddChild(self=Get NamedSlot() -> (NamedSlot:Named Slot Object Reference), Content=@n2)
   > Set Initial Settings!(
    execute
    self:EUW Animation Dialog Object Reference=@n2
    Project:BP Animation Browser Project Object Reference
    Selected Creature:Animatable Creature Object Reference
    Developer Mode:bool=false
    Set Plugin:bool=false
    Animatable Creatures:Array of Animatable Creature Object References
    Show Project Dialog:bool=true
    Override Anim Path Validation:bool=false
   ) -> (then) {is=title K2Node_CallFunction}
   > Bind Event to Generate Animation Paths(self:EUW Animation Dialog Object Reference=@n2, Delegate=Create Event()) {is=title K2Node_AddDelegate}
  ]}
 Create Anim Sequence {nodes [
   n1 = Create Anim Sequence!() -> (
    then
    Level Sequence:Level Sequence Object Reference
    New AnimSequence Path:String
    Binding:Movie Scene Binding Proxy Structure
    Project:BP Animation Browser Project Object Reference
   ) {is=title K2Node_FunctionEntry}
   > PrintString(InString=Append(A="You are trying to create this AnimSeq: ", B=@n1.New AnimSequence Path))
   > n2 = Sequence()
   n3 = Is Valid(exec=@n2.then_1, InputObject=Get Skeleton() -> (Skeleton:Skeleton Object Reference))
   n4 = Get New AnimSequence Path() -> (New AnimSequence Path:String)
   > Execute Python Script!(
    execute
    PythonScript="import unreal\r\n\r\n## Grab the Level Editor World\r\neditor_subsystem = unreal.get_editor_subsystem(unreal.UnrealEditorSubsystem)\r\nworld = editor_subsystem.get_editor_world()\r\n \r\n## Create animation sequence export options\r\nanim_seq_export_options = unreal.AnimSeqExportOption()\r\nanim_seq_export_options.export_transforms = True\r\nanim_seq_export_options.export_morph_targets = True\r\n \r\n## Get asset tools\r\nasset_tools = unreal.AssetToolsHelpers.get_asset_tools()\r\n\r\n## Create an anim sequence factory, assign it the target skeleton\r\nanim_sequence_factory = unreal.AnimSequenceFactory()\r\nanim_sequence_factory.target_skeleton = skeleton\r\nanim_sequence = unreal.AssetTools.create_asset(asset_tools, asset_name=asset_name, package_path=package_path, asset_class=unreal.AnimSequence, factory=anim_sequence_factory)  \r\n\r\n## Export\r\nunreal.SequencerTools.export_anim_sequence(world, level_sequence, anim_sequence, anim_seq_export_options, binding, True)"
    package_path:String=GetPath!(InPath:String=@n4) -> (ReturnValue:String)
    asset_name:String=GetBaseFilename(InPath=@n4, bRemovePath=true)
    level_sequence:Level Sequence Object Reference=Get Level Sequence()
    binding:Movie Scene Binding Proxy Structure=Get Binding() -> (Binding:Movie Scene Binding Proxy Structure)
    skeleton:Skeleton Object Reference=Get Skeleton()
   ) -> (then, ReturnValue) {inputs [package_path, asset_name, level_sequence, binding, skeleton]}
   > n5 = LoadAsset(AssetPath=Get New AnimSequence Path())
   > n6 = Cast To AnimSequence!(execute, Object=@n5) -> (then, CastFailed, AsAnim Sequence:Anim Sequence Object Reference)
   > Set Created AnimSequence(Created AnimSequence:Anim Sequence Object Reference=@n6) -> (Output_Get:Anim Sequence Object Reference)
   > n7 = Sequence()
   n8 = Get Compression Settings!(self:BP Animation Browser Project Object Reference=Get Project() -> (Project:BP Animation Browser Project Object Reference)) -> (Compression Settings:Anim Sequence Soft Object Reference)
   > Branch(Condition=IsValidSoftObjectReference(SoftObjectReference=@n8))
   > n9 = LoadAsset_Blocking(Asset=@n8)
   > n10 = Cast To AnimSequence(Object=@n9)
   > SetBoneCompressionSettings!(
    execute
    AnimationSequence:Anim Sequence Object Reference=Get Created AnimSequence() -> (Created AnimSequence:Anim Sequence Object Reference)
    CompressionSettings:Anim Bone Compression Settings Object Reference=GetBoneCompressionSettings!(AnimationSequence:Anim Sequence Object Reference=@n10) -> (CompressionSettings:Anim Bone Compression Settings Object Reference)
   ) -> (then)
   > SetCurveCompressionSettings!(
    execute
    AnimationSequence:Anim Sequence Object Reference=Get Created AnimSequence()
    CompressionSettings:Anim Curve Compression Settings Object Reference=GetCurveCompressionSettings!(AnimationSequence:Anim Sequence Object Reference=@n10) -> (CompressionSettings:Anim Curve Compression Settings Object Reference)
   ) -> (then)
   > SetAnimationInterpolationType!(
    execute
    AnimationSequence:Anim Sequence Object Reference=Get Created AnimSequence()
    InterpolationType:EAnimInterpolationType Enum=GetAnimationInterpolationType!(AnimationSequence:Anim Sequence Object Reference=@n10) -> (InterpolationType:EAnimInterpolationType Enum=Linear)
   ) -> (then)
   > SaveLoadedAsset(AssetToSave=Get Created AnimSequence(), bOnlyIfIsDirty=false)
   Return Node!(execute=@n7.then_1, Anim Sequence:Anim Sequence Object Reference=Get Created AnimSequence())
   Return Node(execute=@n3.Is Not Valid)
   n11 = Get Current Time!(execute=@n2.then_0, TimeUnit:EMovieSceneTimeUnit Enum=DisplayRate) -> (then, ReturnValue:Movie Scene Sequence Playback Params Structure) {is=title K2Node_CallFunction}
   n12 = Break Movie Scene Sequence Playback Params!(MovieSceneSequencePlaybackParams:"Movie Scene Sequence Playback Params Structure (by ref)"=@n11) -> (
    Frame_FrameNumber_Value:int
    Frame_SubFrame:float=~
    Time:float=~
    MarkedFrame:String
    Timecode:Timecode Structure
    PositionType:EMovieScenePositionType Enum
    UpdateMethod:EUpdatePositionMethod Enum
    bHasJumped:bool=~
   ) {is=title K2Node_BreakStruct}
   > Set Current Time!(
    execute
    PlaybackParams:Movie Scene Sequence Playback Params Structure=Make Movie Scene Sequence Playback Params!(
     Frame_FrameNumber_Value:int=int + int(A=@n12)
     Frame_SubFrame:float=0
     Time:float=0
     MarkedFrame:String
     Timecode:Timecode Structure=(Hours=0,Minutes=0,Seconds=0,Frames=0,Subframe=0,bDropFrameFormat=False)
     PositionType:EMovieScenePositionType Enum=Frame
     UpdateMethod:EUpdatePositionMethod Enum=Play
     bHasJumped:bool=False
    ) -> (MovieSceneSequencePlaybackParams:Movie Scene Sequence Playback Params Structure) {is=title K2Node_MakeStruct}
    TimeUnit:EMovieSceneTimeUnit Enum=DisplayRate
   ) -> (then) {is=title K2Node_CallFunction}
   > Set Current Time(PlaybackParams=@n11, TimeUnit=DisplayRate)
   > n13 = GetBindingID!(
    execute
    Sequence:Movie Scene Sequence Object Reference=Get Level Sequence()
    InBinding:"Movie Scene Binding Proxy Structure (by ref)"=Get Binding()
   ) -> (then, ReturnValue:Movie Scene Object Binding ID Structure)
   > n14 = For Each Loop with Break!(
    Exec:Exec
    Array:Array of Object References=GetBoundObjects!(ObjectBinding:Movie Scene Object Binding ID Structure=@n13) -> (ReturnValue:Array of Object References)
    Break:Exec=@n17.then
   ) -> (LoopBody:Exec, Array Element:Object Reference, Array Index:int, Completed:Exec)
   > n15 = Cast To Actor!(execute, Object=@n14) -> (then, CastFailed, AsActor:Actor Object Reference) {is=title K2Node_DynamicCast}
   n16 = GetComponentByClass!(self:Actor Object Reference=@n15, ComponentClass:Actor Component Class Reference=/Script/Engine.SkeletalMeshComponent) -> (ReturnValue:Skeletal Mesh Component Object Reference)
   > n17 = Set Skeleton(Skeleton:Skeleton Object Reference=Get Skeleton!(self:Skeletal Mesh Object Reference=Get SkeletalMesh!(self:Skinned Mesh Component Object Reference=@n16) -> (SkeletalMesh:Skeletal Mesh Object Reference)) -> (Skeleton)) -> (Output_Get:Skeleton Object Reference)
  ]}
 Create Rows {nodes [
   Create Rows() {is=title K2Node_FunctionEntry}
   > n1 = Sequence!(execute) -> (then_0, then_1, then_2:Exec)
   > n2 = For Each Loop(Array=Get EUW Rows())
   Clear(execute=@n2.Completed, TargetArray:Array of EUW Linked Sequence Row Object References=Get EUW Rows())
   RemoveFromParent(execute=@n2.LoopBody, self=@n2)
   n3 = GetBindings!(execute=@n1.then_1, Sequence:Movie Scene Sequence Object Reference=Get Level Sequence()) -> (then, ReturnValue:Array of Movie Scene Binding Proxy Structures)
   > n4 = GetAnimSequenceLinkFromLevelSequence!(execute, InLevelSequence:Level Sequence Object Reference=Get Level Sequence()) -> (then, ReturnValue:Level Sequence Anim Sequence Link Object Reference)
   > Set Anim Sequence Link(Anim Sequence Link:Level Sequence Anim Sequence Link Object Reference=@n4) -> (Output_Get:Level Sequence Anim Sequence Link Object Reference)
   > n5 = For Each Loop(Array:Array of Movie Scene Binding Proxy Structures=@n3) -> (Array Element:Movie Scene Binding Proxy Structure)
   > Set Local Current Binding(Local Current Binding:Movie Scene Binding Proxy Structure=@n5) -> (Output_Get:Movie Scene Binding Proxy Structure)
   > n6 = Sequence()
   Branch(execute=@n6.then_1, Condition=Get Is Current Binding CR() -> (Is Current Binding CR:bool))
   > n7 = Sequence()
   n8 = For Each Loop!(
    Exec=@n7.then_1
    Array:Array of Level Sequence Anim Sequence Link Item Structures=Get AnimSequenceLinks!(self:Level Sequence Anim Sequence Link Object Reference=Get Anim Sequence Link() -> (Anim Sequence Link:Level Sequence Anim Sequence Link Object Reference)) -> (AnimSequenceLinks:Array of Level Sequence Anim Sequence Link Item Structures)
   ) -> (
    LoopBody
    Array Element_SkelTrackGuid:Guid Structure
    Array Element_PathToAnimSequence:Soft Object Path Structure
    Array Element_bAutoBake:bool=~
    Array Element_bExportTransforms:bool=~
    Array Element_bExportMorphTargets:bool=~
    Array Element_bExportAttributeCurves:bool=~
    Array Element_bExportMaterialCurves:bool=~
    Array Element_bSkipCurvesWithZeroValue:bool=~
    Array Element_bSetRetargetSourceAsset:bool=~
    Array Element_Interpolation:EAnimInterpolationType Enum
    Array Element_CurveInterpolation:ERichCurveInterpMode Enum
    Array Element_bRecordInWorldSpace:bool=~
    Array Element_bEvaluateAllSkeletalMeshComponents:bool=~
    Array Element_IncludeAnimationNames:Array of Strings
    Array Element_ExcludeAnimationNames:Array of Strings
    Array Element_bRemoveExcludedCurves:bool=~
    Array Element_WarmUpFrames:Frame Number Structure
    Array Element_DelayBeforeStart:Frame Number Structure
    Array Element_bUseCustomTimeRange:bool=~
    Array Element_CustomStartFrame:Frame Number Structure
    Array Element_CustomEndFrame:Frame Number Structure
    Array Element_CustomDisplayRate:Frame Rate Structure
    Array Element_bUseCustomFrameRate:bool=~
    Array Element_CustomFrameRate:Frame Rate Structure
    Array Index
    Completed
   )
   > n9 = FindBindingById!(execute, Sequence:Movie Scene Sequence Object Reference=Get Level Sequence(), BindingId:Guid Structure=@n8) -> (then, ReturnValue_BindingID:Guid Structure, ReturnValue_Sequence:Movie Scene Sequence Object Reference)
   n10 = Get Local Current Binding() -> (Local Current Binding:Movie Scene Binding Proxy Structure)
   > Branch(Condition="Equal (Guid)"(
     A:"Guid Structure (by ref)"=@n9
     B:"Guid Structure (by ref)"=Break Movie Scene Binding Proxy!(MovieSceneBindingProxy:"Movie Scene Binding Proxy Structure (by ref)"=@n10) -> (BindingID:Guid Structure, Sequence:Movie Scene Sequence Object Reference) {is=title K2Node_BreakStruct}
    ) -> (ReturnValue:bool=~) {is=title K2Node_PromotableOperator})
   > n11 = LoadAsset(AssetPath=BreakSoftObjectPath(InSoftObjectPath=@n8.Array Element_PathToAnimSequence))
   > n12 = Cast To AnimSequence(Object=@n11)
   > Set AnimSequence(AnimSequence=@n12, self=Get Current Row Widget() -> (Current Row Widget:EUW Linked Sequence Row Object Reference))
   > SetCheckedState!(
    execute
    self:Check Box Object Reference=Get CheckBox!(self:EUW Linked Sequence Row Object Reference=Get Current Row Widget()) -> (CheckBox:Check Box Object Reference)
    InCheckedState:ECheckBoxState Enum=Checked
   ) -> (then)
   n13 = Create Widget!(
    execute:Exec=@n7.then_0
    WorldContextObject:Object Reference
    Class:User Widget Class Reference=/AnimInEdTools/LinkedSequenceManager/EUW/EUW_LinkedSequence_Row.EUW_LinkedSequence_Row_C
    OwningPlayer:Player Controller Object Reference
   ) -> (then:Exec, ReturnValue:EUW Linked Sequence Row Object Reference)
   > Set Current Row Widget(Current Row Widget:EUW Linked Sequence Row Object Reference=@n13) -> (Output_Get:EUW Linked Sequence Row Object Reference)
   > AddChild(self=Get ScrollBox() -> (ScrollBox:Scroll Box Object Reference), Content=Get Current Row Widget())
   > n14 = GetDisplayName!(execute:Exec, InBinding:"Movie Scene Binding Proxy Structure (by ref)"=Get Local Current Binding()) -> (then:Exec, ReturnValue:Text)
   > Set Text(
    Text=@n14
    self=Get BindingName!(self:EUW Linked Sequence Row Object Reference=Get Current Row Widget()) -> (BindingName:Text Object Reference)
   )
   > Set In Binding!(
    execute
    In Binding:Movie Scene Binding Proxy Structure=Get Local Current Binding()
    self:EUW Linked Sequence Row Object Reference=Get Current Row Widget()
   ) -> (then, Output_Get:Movie Scene Binding Proxy Structure)
   > Add(
    TargetArray:Array of EUW Linked Sequence Row Object References=Get EUW Rows()
    NewItem:"EUW Linked Sequence Row Object Reference (by ref)"=Get Current Row Widget()
   ) {is=title K2Node_CallArrayFunction}
   > Bind Event to Create New Anim Sequence For Binding(self:EUW Linked Sequence Row Object Reference=Get Current Row Widget(), Delegate=Create Event()) {is=title K2Node_AddDelegate}
   Update Appearance(execute=@n6.then_2, self=Get Current Row Widget())
   > Set Is Current Binding CR(Is Current Binding CR:bool=false) -> (Output_Get:bool)
   n15 = FindTracksByExactType!(
    execute=@n6.then_0
    InBinding:"Movie Scene Binding Proxy Structure (by ref)"=Get Local Current Binding()
    TrackType:Movie Scene Track Class Reference=/Script/ControlRig.MovieSceneControlRigParameterTrack
   ) -> (then, ReturnValue:Array of Movie Scene Control Rig Parameter Track Object References)
   > n16 = For Each Loop!(Exec, Array:Array of Movie Scene Control Rig Parameter Track Object References=@n15) -> (LoopBody, Array Element:Movie Scene Control Rig Parameter Track Object Reference, Array Index, Completed)
   > n17 = GetSections!(execute, Track:Movie Scene Track Object Reference=@n16) -> (then, ReturnValue:Array of Movie Scene Section Object References)
   > n18 = For Each Loop(Array:Array of Movie Scene Section Object References=@n17) -> (Array Element:Movie Scene Section Object Reference)
   > n19 = Cast To MovieSceneControlRigParameterSection!(execute, Object=@n18) -> (then, CastFailed, AsMovie Scene Control Rig Parameter Section:Movie Scene Control Rig Parameter Section Object Reference) {is=title K2Node_DynamicCast}
   > Set Is Current Binding CR(Is Current Binding CR=IsValid!(Object:Object Reference=@n19) -> (ReturnValue))
  ]}
 EventGraph {nodes [
   n1 = "On Property Changed (LevelSequencePropertyView)"!() -> (OutputDelegate, then, PropertyName:Name=None) {is=title K2Node_ComponentBoundEvent}
   Create Rows(execute=[@n1.then, @n7.then, @n9.then]) {is=title K2Node_CallFunction}
   > Sort Scroll Box By Checkbox()
   "On Clicked (RunButton)"() {is=title K2Node_ComponentBoundEvent}
   > n2 = Sequence!(execute) -> (then_0, then_1)
   > ClearLinkedAnimSequences!(execute, InLevelSequence:Level Sequence Object Reference=Get Level Sequence()) -> (then)
   > n3 = For Each Loop(Array:Array of EUW Linked Sequence Row Object References=Get EUW Rows()) -> (Array Element:EUW Linked Sequence Row Object Reference)
   > Branch(Condition=IsChecked(self=Get CheckBox(self=@n3)))
   > n4 = Execute Python Script!(
    execute
    PythonScript="import unreal\r\n## Create animation sequence export options\r\nanim_seq_export_options = unreal.AnimSeqExportOption()\r\nanim_seq_export_options.export_transforms = True\r\nanim_seq_export_options.export_morph_targets = True"
   ) -> (then, ReturnValue, anim_seq_export_options:Anim Seq Export Option Object Reference) {outputs [anim_seq_export_options]}
   n5 = Get AnimSequence!(self:EUW Linked Sequence Row Object Reference=@n3) -> (AnimSequence:Anim Sequence Object Reference)
   > LinkAnimSequence!(
    execute
    Sequence:Level Sequence Object Reference=Get Level Sequence()
    AnimSequence:Anim Sequence Object Reference=@n5
    ExportOptions:Anim Seq Export Option Object Reference=@n4.anim_seq_export_options
    Binding:"Movie Scene Binding Proxy Structure (by ref)"=Get In Binding(self=@n3)
   ) -> (then, ReturnValue:bool)
   > SaveLoadedAsset(AssetToSave=@n5, bOnlyIfIsDirty=false)
   SaveLoadedAsset(execute=@n3.Completed, AssetToSave=Get Level Sequence(), bOnlyIfIsDirty=false)
   Sort Scroll Box By Checkbox(execute=@n2.then_1)
   "On Clicked (CurrentLevelSequenceButton)"() {is=title K2Node_ComponentBoundEvent}
   > n6 = GetCurrentLevelSequence!(execute) -> (then, ReturnValue:Level Sequence Object Reference)
   > n7 = Set Level Sequence(Level Sequence:Level Sequence Object Reference=@n6) -> (Output_Get:Level Sequence Object Reference)
   Event PreConstruct()
   > SetObject(self=Get LevelSequencePropertyView() -> (LevelSequencePropertyView:Single Property View Object Reference), NewObject=Self-Reference())
   Event Construct()
   > n8 = GetCurrentLevelSequence()
   > n9 = Set Level Sequence(Level Sequence=@n8)
  ]}
 Get Skeletal Bindings {nodes [
   Get Skeletal Bindings!() -> (then, Bindings:Array of Movie Scene Binding Proxy Structures) {is=title K2Node_FunctionEntry}
   > n1 = Sequence!(execute) -> (then_0, then_1, then_2:Exec, then_3:Exec, then_4:Exec)
   Return Node!(
    execute=@n1.then_4
    Skeletal Bindings:Array of Movie Scene Binding Proxy Structures=Get Filtered Skeletal Bindings() -> (Filtered Skeletal Bindings:Array of Movie Scene Binding Proxy Structures)
   )
   > Clear(TargetArray:Array of Movie Scene Binding Proxy Structures=Get Filtered Skeletal Bindings())
   n2 = For Each Loop(
    Exec=@n1.then_1
    Array:Array of Movie Scene Binding Proxy Structures=Get Bindings() -> (Bindings:Array of Movie Scene Binding Proxy Structures)
   ) -> (Array Element:Movie Scene Binding Proxy Structure)
   > n3 = FindTracksByType!(
    execute
    InBinding:"Movie Scene Binding Proxy Structure (by ref)"=@n2
    TrackType:Movie Scene Track Class Reference=/Script/MovieSceneTracks.MovieSceneSkeletalAnimationTrack
   ) -> (then, ReturnValue:Array of Movie Scene Skeletal Animation Track Object References)
   > n4 = For Each Loop(Array:Array of Movie Scene Skeletal Animation Track Object References=@n3) -> (Array Element:Movie Scene Skeletal Animation Track Object Reference)
   > Branch(Condition=IsValid(Object=@n4))
   > Add(
    TargetArray:Array of Movie Scene Binding Proxy Structures=Get Filtered Skeletal Bindings()
    NewItem:"Movie Scene Binding Proxy Structure (by ref)"=@n2
   )
   FindTracksByExactType(InBinding=@n2, TrackType=&MovieSceneSkeletalAnimationTrack) -> (ReturnValue:Array of Movie Scene Skeletal Animation Track Object References)
  ]}
 Sort Scroll Box By Checkbox {nodes [
   Sort Scroll Box By Checkbox() {is=title K2Node_FunctionEntry}
   > Clear(TargetArray:Array of EUW Linked Sequence Row Object References=Get Items To Sort() -> (Items To Sort:Array of EUW Linked Sequence Row Object References))
   n1 = GetAllChildren(self=Get ScrollBox())
   > n2 = For Each Loop(Array:Array of Widget Object References=@n1) -> (Array Element:Widget Object Reference)
   n3 = For Each Loop(Exec=@n2.Completed, Array=@n1)
   > n4 = Cast To EUW_LinkedSequence_Row!(execute, Object=@n3) -> (then, CastFailed, AsEUW Linked Sequence Row:EUW Linked Sequence Row Object Reference) {is=title K2Node_DynamicCast}
   n5 = Get CheckBox(self=@n4)
   > Branch(Condition="Not Equal (Enum)"!(
     A:ECheckBoxState Enum=GetCheckedState!(self:Check Box Object Reference=@n5) -> (ReturnValue:ECheckBoxState Enum=Unchecked)
     B:ECheckBoxState Enum=Checked
    ) -> (ReturnValue:bool))
   > Add(
    TargetArray:Array of EUW Linked Sequence Row Object References=Get Items To Sort()
    NewItem:"EUW Linked Sequence Row Object Reference (by ref)"=@n4
   )
   ClearChildren(execute=@n3.Completed, self=Get ScrollBox())
   > n6 = For Each Loop(Array:Array of EUW Linked Sequence Row Object References=Get Items To Sort()) -> (Array Element:EUW Linked Sequence Row Object Reference)
   > AddChild(self=Get ScrollBox(), Content=@n6)
   n7 = Cast To EUW_LinkedSequence_Row(execute=@n2.LoopBody, Object=@n2)
   n8 = Get CheckBox(self=@n7)
   > Branch(Condition="Not Equal (Enum)"(A=GetCheckedState(self=@n8) -> (ReturnValue=Unchecked), B=Unchecked))
   > Add(TargetArray=Get Items To Sort(), NewItem=@n7)
  ]}
]
variables [
 Current Binding {type=MovieSceneBindingProxy, default {bindingId=#1, sequence=null}}
 EUW Rows {type="array<EUW_LinkedSequence_Row_C>", default []}
 Level Sequence {type=LevelSequence, default=null}
]
widgets {
 info {rootWidgetClass=&CanvasPanel, widgetCount=11}
 widgets [
  CanvasPanel_0 {widgetClassPath=&CanvasPanel, widget=CanvasPanel_0}
  VerticalBox_0 {
   widgetClassPath=&VerticalBox
   parent=CanvasPanel_0
   widget=VerticalBox_0
   slot {class=CanvasPanelSlot, properties {layoutData {alignment=<0 0>, anchors {maximum=<1 1>, minimum=<0 0>}, offsets=<ltrb 0 0 0 0>}}}
  }
  HorizontalBox_0 {widgetClassPath=&HorizontalBox, parent=VerticalBox_0, widget=HorizontalBox_0, slot {class=VerticalBoxSlot}}
  LevelSequencePropertyView {
   widgetClassPath=&SinglePropertyView
   parent=HorizontalBox_0
   widget=LevelSequencePropertyView
   properties {propertyName=Level Sequence}
   slot {class=HorizontalBoxSlot, properties {size {sizeRule=Fill, value=1}}}
  }
  CurrentLevelSequenceButton {
   widgetClassPath=&EditorUtilityButton
   parent=HorizontalBox_0
   widget=CurrentLevelSequenceButton
   slot {class=HorizontalBoxSlot}
  }
  TextBlock_106 {
   widgetClassPath=&TextBlock
   parent=CurrentLevelSequenceButton
   widget=TextBlock_106
   properties {
    text=Get Current LS
    font {
     bForceMonospaced=false
     bMaterialIsStencil=false
     fontMaterial=null
     fontObject=&Roboto
     letterSpacing=0
     monospacedWidth=1
     size=13.5
     skewAmount=0
     typefaceFontName=Bold
     outlineSettings {
      bApplyOutlineToDropShadows=false
      bMiteredCorners=false
      bSeparateFillAlpha=false
      outlineMaterial=null
      outlineSize=0
      outlineColor=<rgba 0 0 0 1>
     }
    }
   }
   slot {class=ButtonSlot}
  }
  ScrollBox {
   widgetClassPath=&ScrollBox
   bIsVariable=true
   parent=VerticalBox_0
   widget=ScrollBox
   properties {
    consumeMouseWheel=Always
    widgetBarStyle {
     thickness=8
     draggedThumbImage {
      bIsDynamicallyLoaded=false
      drawAs=RoundedBox
      imageType=NoImage
      mirroring=NoMirror
      resourceName=None
      resourceObject=null
      tiling=NoTile
      imageSize=<0 0>
      margin=<ltrb 0 0 0 0>
      outlineSettings {
       bUseBrushTransparency=false
       roundingType=FixedRadius
       width=0
       color {colorUseRule=UseColor_Specified, specifiedColor=<rgba 0 0 0 0>}
       cornerRadii=<4 4 4 4>
      }
      tintColor {colorUseRule=UseColor_Specified, specifiedColor=<rgba 0.2159 0.2159 0.2159 1>}
      uVRegion {bIsValid=false, max=<0 0>, min=<0 0>}
     }
     horizontalBackgroundImage {
      bIsDynamicallyLoaded=false
      drawAs=NoDrawType
      imageType=NoImage
      mirroring=NoMirror
      resourceName=None
      resourceObject=null
      tiling=NoTile
      imageSize=<0 0>
      margin=<ltrb 0 0 0 0>
      outlineSettings {
       bUseBrushTransparency=false
       roundingType=HalfHeightRadius
       width=0
       color {colorUseRule=UseColor_Specified, specifiedColor=<rgba 0 0 0 0>}
       cornerRadii=<0 0 0 0>
      }
      tintColor {colorUseRule=UseColor_Specified, specifiedColor=<rgba 1 1 1 1>}
      uVRegion {bIsValid=false, max=<0 0>, min=<0 0>}
     }
     horizontalBottomSlotImage {
      bIsDynamicallyLoaded=false
      drawAs=NoDrawType
      imageType=NoImage
      mirroring=NoMirror
      resourceName=None
      resourceObject=null
      tiling=NoTile
      imageSize=<0 0>
      margin=<ltrb 0 0 0 0>
      outlineSettings {
       bUseBrushTransparency=false
       roundingType=HalfHeightRadius
       width=0
       color {colorUseRule=UseColor_Specified, specifiedColor=<rgba 0 0 0 0>}
       cornerRadii=<0 0 0 0>
      }
      tintColor {colorUseRule=UseColor_Specified, specifiedColor=<rgba 1 1 1 1>}
      uVRegion {bIsValid=false, max=<0 0>, min=<0 0>}
     }
     horizontalTopSlotImage {
      bIsDynamicallyLoaded=false
      drawAs=NoDrawType
      imageType=NoImage
      mirroring=NoMirror
      resourceName=None
      resourceObject=null
      tiling=NoTile
      imageSize=<0 0>
      margin=<ltrb 0 0 0 0>
      outlineSettings {
       bUseBrushTransparency=false
       roundingType=HalfHeightRadius
       width=0
       color {colorUseRule=UseColor_Specified, specifiedColor=<rgba 0 0 0 0>}
       cornerRadii=<0 0 0 0>
      }
      tintColor {colorUseRule=UseColor_Specified, specifiedColor=<rgba 1 1 1 1>}
      uVRegion {bIsValid=false, max=<0 0>, min=<0 0>}
     }
     hoveredThumbImage {
      bIsDynamicallyLoaded=false
      drawAs=RoundedBox
      imageType=NoImage
      mirroring=NoMirror
      resourceName=None
      resourceObject=null
      tiling=NoTile
      imageSize=<0 0>
      margin=<ltrb 0 0 0 0>
      outlineSettings {
       bUseBrushTransparency=false
       roundingType=FixedRadius
       width=0
       color {colorUseRule=UseColor_Specified, specifiedColor=<rgba 0 0 0 0>}
       cornerRadii=<4 4 4 4>
      }
      tintColor {colorUseRule=UseColor_Specified, specifiedColor=<rgba 0.2159 0.2159 0.2159 1>}
      uVRegion {bIsValid=false, max=<0 0>, min=<0 0>}
     }
     normalThumbImage {
      bIsDynamicallyLoaded=false
      drawAs=RoundedBox
      imageType=NoImage
      mirroring=NoMirror
      resourceName=None
      resourceObject=null
      tiling=NoTile
      imageSize=<0 0>
      margin=<ltrb 0 0 0 0>
      outlineSettings {
       bUseBrushTransparency=false
       roundingType=FixedRadius
       width=0
       color {colorUseRule=UseColor_Specified, specifiedColor=<rgba 0 0 0 0>}
       cornerRadii=<4 4 4 4>
      }
      tintColor {colorUseRule=UseColor_Specified, specifiedColor=<rgba 0.0953 0.0953 0.0953 1>}
      uVRegion {bIsValid=false, max=<0 0>, min=<0 0>}
     }
     verticalBackgroundImage {
      bIsDynamicallyLoaded=false
      drawAs=NoDrawType
      imageType=NoImage
      mirroring=NoMirror
      resourceName=None
      resourceObject=null
      tiling=NoTile
      imageSize=<0 0>
      margin=<ltrb 0 0 0 0>
      outlineSettings {
       bUseBrushTransparency=false
       roundingType=HalfHeightRadius
       width=0
       color {colorUseRule=UseColor_Specified, specifiedColor=<rgba 0 0 0 0>}
       cornerRadii=<0 0 0 0>
      }
      tintColor {colorUseRule=UseColor_Specified, specifiedColor=<rgba 1 1 1 1>}
      uVRegion {bIsValid=false, max=<0 0>, min=<0 0>}
     }
     verticalBottomSlotImage {
      bIsDynamicallyLoaded=false
      drawAs=NoDrawType
      imageType=NoImage
      mirroring=NoMirror
      resourceName=None
      resourceObject=null
      tiling=NoTile
      imageSize=<0 0>
      margin=<ltrb 0 0 0 0>
      outlineSettings {
       bUseBrushTransparency=false
       roundingType=HalfHeightRadius
       width=0
       color {colorUseRule=UseColor_Specified, specifiedColor=<rgba 0 0 0 0>}
       cornerRadii=<0 0 0 0>
      }
      tintColor {colorUseRule=UseColor_Specified, specifiedColor=<rgba 1 1 1 1>}
      uVRegion {bIsValid=false, max=<0 0>, min=<0 0>}
     }
     verticalTopSlotImage {
      bIsDynamicallyLoaded=false
      drawAs=NoDrawType
      imageType=NoImage
      mirroring=NoMirror
      resourceName=None
      resourceObject=null
      tiling=NoTile
      imageSize=<0 0>
      margin=<ltrb 0 0 0 0>
      outlineSettings {
       bUseBrushTransparency=false
       roundingType=HalfHeightRadius
       width=0
       color {colorUseRule=UseColor_Specified, specifiedColor=<rgba 0 0 0 0>}
       cornerRadii=<0 0 0 0>
      }
      tintColor {colorUseRule=UseColor_Specified, specifiedColor=<rgba 1 1 1 1>}
      uVRegion {bIsValid=false, max=<0 0>, min=<0 0>}
     }
    }
    widgetStyle {
     barThickness=8
     bottomShadowBrush {
      bIsDynamicallyLoaded=false
      drawAs=Box
      imageType=FullColor
      mirroring=NoMirror
      resourceName=../../../Engine/Content/Slate/Common/ScrollBoxShadowBottom.png
      resourceObject=null
      tiling=NoTile
      imageSize=<16 8>
      margin=<ltrb 0.5 0 0.5 1>
      outlineSettings {
       bUseBrushTransparency=false
       roundingType=HalfHeightRadius
       width=0
       color {colorUseRule=UseColor_Specified, specifiedColor=<rgba 0 0 0 0>}
       cornerRadii=<0 0 0 0>
      }
      tintColor {colorUseRule=UseColor_Specified, specifiedColor=<rgba 1 1 1 1>}
      uVRegion {bIsValid=false, max=<0 0>, min=<0 0>}
     }
     horizontalScrolledContentPadding=<ltrb 0 0 1 0>
     leftShadowBrush {
      bIsDynamicallyLoaded=false
      drawAs=Box
      imageType=FullColor
      mirroring=NoMirror
      resourceName=../../../Engine/Content/Slate/Common/ScrollBoxShadowLeft.png
      resourceObject=null
      tiling=NoTile
      imageSize=<8 16>
      margin=<ltrb 1 0.5 0 0.5>
      outlineSettings {
       bUseBrushTransparency=false
       roundingType=HalfHeightRadius
       width=0
       color {colorUseRule=UseColor_Specified, specifiedColor=<rgba 0 0 0 0>}
       cornerRadii=<0 0 0 0>
      }
      tintColor {colorUseRule=UseColor_Specified, specifiedColor=<rgba 1 1 1 1>}
      uVRegion {bIsValid=false, max=<0 0>, min=<0 0>}
     }
     rightShadowBrush {
      bIsDynamicallyLoaded=false
      drawAs=Box
      imageType=FullColor
      mirroring=NoMirror
      resourceName=../../../Engine/Content/Slate/Common/ScrollBoxShadowRight.png
      resourceObject=null
      tiling=NoTile
      imageSize=<8 16>
      margin=<ltrb 0 0.5 1 0.5>
      outlineSettings {
       bUseBrushTransparency=false
       roundingType=HalfHeightRadius
       width=0
       color {colorUseRule=UseColor_Specified, specifiedColor=<rgba 0 0 0 0>}
       cornerRadii=<0 0 0 0>
      }
      tintColor {colorUseRule=UseColor_Specified, specifiedColor=<rgba 1 1 1 1>}
      uVRegion {bIsValid=false, max=<0 0>, min=<0 0>}
     }
     topShadowBrush {
      bIsDynamicallyLoaded=false
      drawAs=Box
      imageType=FullColor
      mirroring=NoMirror
      resourceName=../../../Engine/Content/Slate/Common/ScrollBoxShadowTop.png
      resourceObject=null
      tiling=NoTile
      imageSize=<16 8>
      margin=<ltrb 0.5 1 0.5 0>
      outlineSettings {
       bUseBrushTransparency=false
       roundingType=HalfHeightRadius
       width=0
       color {colorUseRule=UseColor_Specified, specifiedColor=<rgba 0 0 0 0>}
       cornerRadii=<0 0 0 0>
      }
      tintColor {colorUseRule=UseColor_Specified, specifiedColor=<rgba 1 1 1 1>}
      uVRegion {bIsValid=false, max=<0 0>, min=<0 0>}
     }
     verticalScrolledContentPadding=<ltrb 0 0 0 1>
    }
   }
   slot {class=VerticalBoxSlot, properties {size {sizeRule=Fill, value=1}}}
  }
  RunButton {
   widgetClassPath=&Button
   parent=VerticalBox_0
   widget=RunButton
   properties {widgetStyle {
     clickedSlateSound {resourceObject=null}
     disabled {
      bIsDynamicallyLoaded=false
      drawAs=RoundedBox
      imageType=NoImage
      mirroring=NoMirror
      resourceName=None
      resourceObject=null
      tiling=NoTile
      imageSize=<0 0>
      margin=<ltrb 0 0 0 0>
      outlineSettings {
       bUseBrushTransparency=false
       roundingType=FixedRadius
       width=1
       color {colorUseRule=UseColor_ColorTable, specifiedColor=<rgba 0 0 0 0>}
       cornerRadii=<4 4 4 4>
      }
      tintColor {colorUseRule=UseColor_Specified, specifiedColor=<rgba 0.0395 0.0395 0.0395 1>}
      uVRegion {bIsValid=false, max=<0 0>, min=<0 0>}
     }
     disabledForeground {colorUseRule=UseColor_ColorTable, specifiedColor=<rgba 0 0 0 0>}
     hovered {
      bIsDynamicallyLoaded=false
      drawAs=RoundedBox
      imageType=NoImage
      mirroring=NoMirror
      resourceName=None
      resourceObject=null
      tiling=NoTile
      imageSize=<32 32>
      margin=<ltrb 0 0 0 0>
      outlineSettings {
       bUseBrushTransparency=false
       roundingType=FixedRadius
       width=1
       color {colorUseRule=UseColor_ColorTable, specifiedColor=<rgba 0 0 0 0>}
       cornerRadii=<4 4 4 4>
      }
      tintColor {colorUseRule=UseColor_Specified, specifiedColor=<rgba 0.0953 0.0953 0.0953 1>}
      uVRegion {bIsValid=false, max=<0 0>, min=<0 0>}
     }
     hoveredForeground {colorUseRule=UseColor_ColorTable, specifiedColor=<rgba 0 0 0 0>}
     hoveredSlateSound {resourceObject=null}
     normal {
      bIsDynamicallyLoaded=false
      drawAs=RoundedBox
      imageType=NoImage
      mirroring=NoMirror
      resourceName=None
      resourceObject=null
      tiling=NoTile
      imageSize=<32 32>
      margin=<ltrb 0 0 0 0>
      outlineSettings {
       bUseBrushTransparency=false
       roundingType=FixedRadius
       width=1
       color {colorUseRule=UseColor_ColorTable, specifiedColor=<rgba 0 0 0 0>}
       cornerRadii=<4 4 4 4>
      }
      tintColor {colorUseRule=UseColor_Specified, specifiedColor=<rgba 0.0395 0.0395 0.0395 1>}
      uVRegion {bIsValid=false, max=<0 0>, min=<0 0>}
     }
     normalForeground {colorUseRule=UseColor_ColorTable, specifiedColor=<rgba 0 0 0 0>}
     normalPadding=<ltrb 12 1.5 12 1.5>
     pressed {
      bIsDynamicallyLoaded=false
      drawAs=RoundedBox
      imageType=NoImage
      mirroring=NoMirror
      resourceName=None
      resourceObject=null
      tiling=NoTile
      imageSize=<32 32>
      margin=<ltrb 0 0 0 0>
      outlineSettings {
       bUseBrushTransparency=false
       roundingType=FixedRadius
       width=1
       color {colorUseRule=UseColor_ColorTable, specifiedColor=<rgba 0 0 0 0>}
       cornerRadii=<4 4 4 4>
      }
      tintColor {colorUseRule=UseColor_Specified, specifiedColor=<rgba 0.0284 0.0284 0.0284 1>}
      uVRegion {bIsValid=false, max=<0 0>, min=<0 0>}
     }
     pressedForeground {colorUseRule=UseColor_ColorTable, specifiedColor=<rgba 0 0 0 0>}
     pressedPadding=<ltrb 12 2.5 12 0.5>
     pressedSlateSound {resourceObject=null}
    }}
   slot {class=VerticalBoxSlot}
  }
  TextBlock_4 {widgetClassPath=&TextBlock, parent=RunButton, widget=TextBlock_4, properties {text=Run}, slot {class=ButtonSlot}}
  Overlay {
   widgetClassPath=&Overlay
   bIsVariable=true
   parent=CanvasPanel_0
   widget=Overlay
   slot {class=CanvasPanelSlot, properties {zOrder=1, layoutData {alignment=<0 0>, anchors {maximum=<1 1>, minimum=<0 0>}, offsets=<ltrb 0 0 0 0>}}}
  }
  NamedSlot {
   widgetClassPath=&NamedSlot
   bIsVariable=true
   parent=Overlay
   widget=NamedSlot
   slot {class=OverlaySlot, properties {horizontalAlignment=HAlign_Fill, verticalAlignment=VAlign_Fill}}
  }
 ]
}
