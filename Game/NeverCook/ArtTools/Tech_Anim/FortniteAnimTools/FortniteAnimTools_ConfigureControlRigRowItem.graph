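# Editor utility row widget for Fortnite anim tools that shows a Sequencer binding's component name and Control Rig tracks, with a button to remove them.
# refresh_ui, run on Construct, resolves binding_id in the focused Level Sequence and joins its Control Rig track display names into the label.
# Remove Rigs deletes every MovieSceneControlRigParameterTrack on that binding, resets the label to None, then refreshes.
# The Add Rigs button is collapsed and has no logic.
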EditorUtilityWidgetBlueprint /Game/NeverCook/ArtTools/Tech_Anim/FortniteAnimTools/FortniteAnimTools_ConfigureControlRigRowItem
graphs [
 EventGraph {nodes [
   n1 = GetCurrentLevelSequence!(execute) -> (then, ReturnValue:Level Sequence Object Reference)
   n2 = GetFocusedLevelSequence!(execute=@n7.then) -> (then, ReturnValue:Level Sequence Object Reference)
   n3 = ResolveBindingID!(
    execute=[@n1.then, @n2.then]
    RootSequence:Movie Scene Sequence Object Reference=@n2
    InObjectBindingID:Movie Scene Object Binding ID Structure=Get binding_id() -> (binding_id:Movie Scene Object Binding ID Structure)
   ) -> (then, ReturnValue:Movie Scene Binding Proxy Structure)
   > n4 = GetTracks!(execute, InBinding:"Movie Scene Binding Proxy Structure (by ref)"=@n3) -> (then, ReturnValue:Array of Movie Scene Track Object References)
   > n5 = For Each Loop(Array:Array of Movie Scene Track Object References=@n4) -> (Array Element:Movie Scene Track Object Reference)
   Set Text(execute=@n5.Completed, Text="LOCTEXT(\"None\")", self=Get compoennt_name_txt_1() -> (compoennt_name_txt_1:Text Object Reference))
   > refresh_ui()
   n6 = Cast To MovieSceneControlRigParameterTrack!(execute=@n5.LoopBody, Object=@n5) -> (then, CastFailed, AsMovie Scene Control Rig Parameter Track:Movie Scene Control Rig Parameter Track Object Reference) {is=title K2Node_DynamicCast}
   > RemoveTrack!(execute, InBinding:"Movie Scene Binding Proxy Structure (by ref)"=@n3, TrackToRemove:Movie Scene Track Object Reference=@n6) -> (then)
   n7 = "On Clicked (remove_rig_btn)"() {is=title K2Node_ComponentBoundEvent}
   Event Construct()
   > refresh_ui()
   Event PreConstruct()
  ]}
 refresh_ui {nodes [
   n1 = GetCurrentLevelSequence()
   n2 = GetFocusedLevelSequence(execute=@n9.then)
   n3 = ResolveBindingID(execute=[@n2.then, @n1.then], RootSequence=@n2, InObjectBindingID=Get binding_id())
   > Clear(TargetArray=Get control_rig_names() -> (control_rig_names:Array of Strings))
   > n4 = GetTracks(InBinding=@n3)
   > n5 = For Each Loop(Array=@n4)
   n6 = Get control_rig_names()
   Branch(execute=@n5.Completed, Condition="integer > integer"(A=Length(TargetArray:Array of Strings=@n6)))
   > Set Text(Text="To Text (String)"(InString=JoinStringArray(SourceArray=@n6, Separator=" ")), self=Get compoennt_name_txt_1())
   n7 = Cast To MovieSceneControlRigParameterTrack(execute=@n5.LoopBody, Object=@n5)
   > n8 = GetTrackRowDisplayName!(execute, Track:Movie Scene Track Object Reference=@n7, RowIndex:int=0) -> (then, ReturnValue:Text)
   > Add(TargetArray=@n6, NewItem="To String (Text)"(InText=@n8)) {is=title K2Node_CallArrayFunction}
   refresh_ui() {is=title K2Node_FunctionEntry}
   > n9 = Set Text(Text=Get component_name() -> (component_name:Text), self=Get compoennt_name_txt() -> (compoennt_name_txt:Text Object Reference))
  ]}
]
variables [
 binding_id {type=MovieSceneObjectBindingID, default {guid=#1, resolveParentIndex=-1, sequenceId=0}}
 component_name {type=string, default=Component}
 control_rig_names {type="array<string>", default []}
]
widgets {
 info {rootWidgetClass=&Border, widgetCount=13}
 widgets [
  Border_2 {widgetClassPath=&Border, widget=Border_2, properties {brushColor=<rgba 0.0073 0.0073 0.0073 0.9>}}
  SizeBox_2 {
   widgetClassPath=&SizeBox
   parent=Border_2
   widget=SizeBox_2
   properties {bOverride_MinDesiredWidth=true, minDesiredWidth=600}
   slot {class=BorderSlot}
  }
  mannequin_settings_vertical_layout {
   widgetClassPath=&VerticalBox
   bIsVariable=true
   parent=SizeBox_2
   widget=mannequin_settings_vertical_layout
   properties {clipping=ClipToBounds, visibility=Visible}
   slot {class=SizeBoxSlot, properties {verticalAlignment=VAlign_Center}}
  }
  Border_3 {
   widgetClassPath=&Border
   parent=mannequin_settings_vertical_layout
   widget=Border_3
   properties {
    background {
     bIsDynamicallyLoaded=false
     drawAs=Box
     imageType=NoImage
     mirroring=NoMirror
     resourceName=None
     resourceObject=null
     tiling=NoTile
     imageSize=<32 32>
     margin=<ltrb 0 0 0 0>
     outlineSettings {
      bUseBrushTransparency=false
      roundingType=HalfHeightRadius
      width=0
      color {colorUseRule=UseColor_Specified, specifiedColor=<rgba 0 0 0 0>}
      cornerRadii=<0 0 0 0>
     }
     tintColor {colorUseRule=UseColor_Specified, specifiedColor=<rgba 1 1 1 1>}
     uVRegion {bIsValid=false, max=<0 0>, min=<0 0>}
    }
    brushColor=<rgba 0.0885 0.0885 0.0885 1>
    padding=<ltrb 0 0 0 0>
   }
   slot {class=VerticalBoxSlot}
  }
  HorizontalBox_14 {
   widgetClassPath=&HorizontalBox
   parent=Border_3
   widget=HorizontalBox_14
   slot {class=BorderSlot, properties {padding=<ltrb 0 0 0 0>}}
  }
  Border_0 {
   widgetClassPath=&Border
   parent=HorizontalBox_14
   widget=Border_0
   properties {verticalAlignment=VAlign_Center, brushColor=<rgba 0.0677 0.0677 0.0677 1>}
   slot {class=HorizontalBoxSlot, properties {padding=<ltrb 0 0 10 0>}}
  }
  compoennt_name_txt {
   widgetClassPath=&TextBlock
   bIsVariable=true
   parent=Border_0
   widget=compoennt_name_txt
   properties {
    clipping=ClipToBounds
    text=Component Name
    font {
     bForceMonospaced=false
     bMaterialIsStencil=false
     fontMaterial=null
     fontObject=&Roboto
     letterSpacing=0
     monospacedWidth=1
     size=12
     skewAmount=0
     typefaceFontName=Regular
     outlineSettings {
      bApplyOutlineToDropShadows=false
      bMiteredCorners=false
      bSeparateFillAlpha=false
      outlineMaterial=null
      outlineSize=0
      outlineColor=<rgba 0 0 0 1>
     }
    }
   }
   slot {class=BorderSlot, properties {verticalAlignment=VAlign_Center}}
  }
  Border_1 {
   widgetClassPath=&Border
   parent=HorizontalBox_14
   widget=Border_1
   properties {brushColor=<rgba 1 1 1 0>}
   slot {class=HorizontalBoxSlot, properties {verticalAlignment=VAlign_Center, padding=<ltrb 10 0 10 0>, size {sizeRule=Fill, value=1}}}
  }
  compoennt_name_txt_1 {
   widgetClassPath=&TextBlock
   bIsVariable=true
   parent=Border_1
   widget=compoennt_name_txt_1
   properties {
    clipping=ClipToBounds
    justification=Center
    text=None
    font {
     bForceMonospaced=false
     bMaterialIsStencil=false
     fontMaterial=null
     fontObject=&Roboto
     letterSpacing=0
     monospacedWidth=1
     size=10.5
     skewAmount=0
     typefaceFontName=Regular
     outlineSettings {
      bApplyOutlineToDropShadows=false
      bMiteredCorners=false
      bSeparateFillAlpha=false
      outlineMaterial=null
      outlineSize=0
      outlineColor=<rgba 0 0 0 1>
     }
    }
   }
   slot {class=BorderSlot}
  }
  add_rig_btn {
   widgetClassPath=&EditorUtilityButton
   parent=HorizontalBox_14
   widget=add_rig_btn
   properties {
    visibility=Collapsed
    widgetStyle {
     clickedSlateSound {resourceObject=null}
     disabled {
      bIsDynamicallyLoaded=false
      drawAs=RoundedBox
      imageType=NoImage
      mirroring=NoMirror
      resourceName=None
      resourceObject=null
      tiling=NoTile
      imageSize=<0 0>
      margin=<ltrb 0 0 0 0>
      outlineSettings {
       bUseBrushTransparency=false
       roundingType=FixedRadius
       width=1
       color {colorUseRule=UseColor_ColorTable, specifiedColor=<rgba 0 0 0 0>}
       cornerRadii=<4 4 4 4>
      }
      tintColor {colorUseRule=UseColor_Specified, specifiedColor=<rgba 0.0395 0.0395 0.0395 1>}
      uVRegion {bIsValid=false, max=<0 0>, min=<0 0>}
     }
     disabledForeground {colorUseRule=UseColor_ColorTable, specifiedColor=<rgba 0 0 0 0>}
     hovered {
      bIsDynamicallyLoaded=false
      drawAs=RoundedBox
      imageType=NoImage
      mirroring=NoMirror
      resourceName=None
      resourceObject=null
      tiling=NoTile
      imageSize=<32 32>
      margin=<ltrb 0 0 0 0>
      outlineSettings {
       bUseBrushTransparency=false
       roundingType=FixedRadius
       width=1
       color {colorUseRule=UseColor_ColorTable, specifiedColor=<rgba 0 0 0 0>}
       cornerRadii=<4 4 4 4>
      }
      tintColor {colorUseRule=UseColor_Specified, specifiedColor=<rgba 0 0.6927 0.1824 1>}
      uVRegion {bIsValid=false, max=<0 0>, min=<0 0>}
     }
     hoveredForeground {colorUseRule=UseColor_ColorTable, specifiedColor=<rgba 0 0 0 0>}
     hoveredSlateSound {resourceObject=null}
     normal {
      bIsDynamicallyLoaded=false
      drawAs=RoundedBox
      imageType=NoImage
      mirroring=NoMirror
      resourceName=None
      resourceObject=null
      tiling=NoTile
      imageSize=<32 32>
      margin=<ltrb 0 0 0 0>
      outlineSettings {
       bUseBrushTransparency=false
       roundingType=FixedRadius
       width=1
       color {colorUseRule=UseColor_ColorTable, specifiedColor=<rgba 0 0 0 0>}
       cornerRadii=<4 4 4 4>
      }
      tintColor {colorUseRule=UseColor_Specified, specifiedColor=<rgba 0 0.4635 0.1159 1>}
      uVRegion {bIsValid=false, max=<0 0>, min=<0 0>}
     }
     normalForeground {colorUseRule=UseColor_ColorTable, specifiedColor=<rgba 0 0 0 0>}
     normalPadding=<ltrb 12 1.5 12 1.5>
     pressed {
      bIsDynamicallyLoaded=false
      drawAs=RoundedBox
      imageType=NoImage
      mirroring=NoMirror
      resourceName=None
      resourceObject=null
      tiling=NoTile
      imageSize=<32 32>
      margin=<ltrb 0 0 0 0>
      outlineSettings {
       bUseBrushTransparency=false
       roundingType=FixedRadius
       width=1
       color {colorUseRule=UseColor_ColorTable, specifiedColor=<rgba 0 0 0 0>}
       cornerRadii=<4 4 4 4>
      }
      tintColor {colorUseRule=UseColor_Specified, specifiedColor=<rgba 0.0284 0.0284 0.0284 1>}
      uVRegion {bIsValid=false, max=<0 0>, min=<0 0>}
     }
     pressedForeground {colorUseRule=UseColor_ColorTable, specifiedColor=<rgba 0 0 0 0>}
     pressedPadding=<ltrb 12 2.5 12 0.5>
     pressedSlateSound {resourceObject=null}
    }
   }
   slot {class=HorizontalBoxSlot, properties {horizontalAlignment=HAlign_Right, verticalAlignment=VAlign_Top, padding=<ltrb 5 5 5 5>}}
  }
  TextBlock_10 {
   widgetClassPath=&TextBlock
   parent=add_rig_btn
   widget=TextBlock_10
   properties {
    text=Add Rigs
    font {
     bForceMonospaced=false
     bMaterialIsStencil=false
     fontMaterial=null
     fontObject=&Roboto
     letterSpacing=0
     monospacedWidth=1
     size=9
     skewAmount=0
     typefaceFontName=Bold
     outlineSettings {
      bApplyOutlineToDropShadows=false
      bMiteredCorners=false
      bSeparateFillAlpha=false
      outlineMaterial=null
      outlineSize=0
      outlineColor=<rgba 0 0 0 1>
     }
    }
   }
   slot {class=ButtonSlot, properties {horizontalAlignment=HAlign_Fill, verticalAlignment=VAlign_Fill, padding=<ltrb 5 5 5 5>}}
  }
  remove_rig_btn {
   widgetClassPath=&EditorUtilityButton
   parent=HorizontalBox_14
   widget=remove_rig_btn
   properties {widgetStyle {
     clickedSlateSound {resourceObject=null}
     disabled {
      bIsDynamicallyLoaded=false
      drawAs=RoundedBox
      imageType=NoImage
      mirroring=NoMirror
      resourceName=None
      resourceObject=null
      tiling=NoTile
      imageSize=<0 0>
      margin=<ltrb 0 0 0 0>
      outlineSettings {
       bUseBrushTransparency=false
       roundingType=FixedRadius
       width=1
       color {colorUseRule=UseColor_ColorTable, specifiedColor=<rgba 0 0 0 0>}
       cornerRadii=<4 4 4 4>
      }
      tintColor {colorUseRule=UseColor_Specified, specifiedColor=<rgba 0.0395 0.0395 0.0395 1>}
      uVRegion {bIsValid=false, max=<0 0>, min=<0 0>}
     }
     disabledForeground {colorUseRule=UseColor_ColorTable, specifiedColor=<rgba 0 0 0 0>}
     hovered {
      bIsDynamicallyLoaded=false
      drawAs=RoundedBox
      imageType=NoImage
      mirroring=NoMirror
      resourceName=None
      resourceObject=null
      tiling=NoTile
      imageSize=<32 32>
      margin=<ltrb 0 0 0 0>
      outlineSettings {
       bUseBrushTransparency=false
       roundingType=FixedRadius
       width=1
       color {colorUseRule=UseColor_ColorTable, specifiedColor=<rgba 0 0 0 0>}
       cornerRadii=<4 4 4 4>
      }
      tintColor {colorUseRule=UseColor_Specified, specifiedColor=<rgba 0.4635 0.0579 0.0579 1>}
      uVRegion {bIsValid=false, max=<0 0>, min=<0 0>}
     }
     hoveredForeground {colorUseRule=UseColor_ColorTable, specifiedColor=<rgba 0 0 0 0>}
     hoveredSlateSound {resourceObject=null}
     normal {
      bIsDynamicallyLoaded=false
      drawAs=RoundedBox
      imageType=NoImage
      mirroring=NoMirror
      resourceName=None
      resourceObject=null
      tiling=NoTile
      imageSize=<32 32>
      margin=<ltrb 0 0 0 0>
      outlineSettings {
       bUseBrushTransparency=false
       roundingType=FixedRadius
       width=1
       color {colorUseRule=UseColor_ColorTable, specifiedColor=<rgba 0 0 0 0>}
       cornerRadii=<4 4 4 4>
      }
      tintColor {colorUseRule=UseColor_Specified, specifiedColor=<rgba 0.4635 0.1207 0.1207 1>}
      uVRegion {bIsValid=false, max=<0 0>, min=<0 0>}
     }
     normalForeground {colorUseRule=UseColor_ColorTable, specifiedColor=<rgba 0 0 0 0>}
     normalPadding=<ltrb 12 1.5 12 1.5>
     pressed {
      bIsDynamicallyLoaded=false
      drawAs=RoundedBox
      imageType=NoImage
      mirroring=NoMirror
      resourceName=None
      resourceObject=null
      tiling=NoTile
      imageSize=<32 32>
      margin=<ltrb 0 0 0 0>
      outlineSettings {
       bUseBrushTransparency=false
       roundingType=FixedRadius
       width=1
       color {colorUseRule=UseColor_ColorTable, specifiedColor=<rgba 0 0 0 0>}
       cornerRadii=<4 4 4 4>
      }
      tintColor {colorUseRule=UseColor_Specified, specifiedColor=<rgba 0.0284 0.0284 0.0284 1>}
      uVRegion {bIsValid=false, max=<0 0>, min=<0 0>}
     }
     pressedForeground {colorUseRule=UseColor_ColorTable, specifiedColor=<rgba 0 0 0 0>}
     pressedPadding=<ltrb 12 2.5 12 0.5>
     pressedSlateSound {resourceObject=null}
    }}
   slot {class=HorizontalBoxSlot, properties {horizontalAlignment=HAlign_Right, verticalAlignment=VAlign_Top, padding=<ltrb 5 5 5 5>}}
  }
  TextBlock {
   widgetClassPath=&TextBlock
   parent=remove_rig_btn
   widget=TextBlock
   properties {
    text=Remove Rigs
    font {
     bForceMonospaced=false
     bMaterialIsStencil=false
     fontMaterial=null
     fontObject=&Roboto
     letterSpacing=0
     monospacedWidth=1
     size=9
     skewAmount=0
     typefaceFontName=Bold
     outlineSettings {
      bApplyOutlineToDropShadows=false
      bMiteredCorners=false
      bSeparateFillAlpha=false
      outlineMaterial=null
      outlineSize=0
      outlineColor=<rgba 0 0 0 1>
     }
    }
   }
   slot {class=ButtonSlot, properties {horizontalAlignment=HAlign_Fill, verticalAlignment=VAlign_Fill, padding=<ltrb 5 5 5 5>}}
  }
 ]
}
