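# Editor utility widget in the Fortnite anim tools: a "Mannequin Config" panel for configuring a Sequencer-bound FortPlayerMannequin.
# On Construct it resolves the first object bound to binding_id and shows selected variant, back bling, character part and weapon settings in a details view.
# If the mannequin cast succeeds, it adds an "Actor" Control Rig row widget to a grid.
# Per-skeletal-mesh rows and sequence binding scans exist but are not wired to events.
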EditorUtilityWidgetBlueprint /Game/NeverCook/ArtTools/Tech_Anim/FortniteAnimTools/FortniteAnimTools_ConfigureActor
graphs [EventGraph {nodes [
   n1 = GetFocusedLevelSequence!(execute) -> (then, ReturnValue:Level Sequence Object Reference)
   n2 = GetBindings!(execute=[@n1.then, @n12.then], Sequence:Movie Scene Sequence Object Reference=@n1) -> (then, ReturnValue:Array of Movie Scene Binding Proxy Structures)
   > n3 = For Each Loop(Array:Array of Movie Scene Binding Proxy Structures=@n2) -> (Array Element:Movie Scene Binding Proxy Structure)
   > n4 = GetBindingID!(execute, Sequence:Movie Scene Sequence Object Reference, InBinding:"Movie Scene Binding Proxy Structure (by ref)"=@n3) -> (then, ReturnValue:Movie Scene Object Binding ID Structure)
   > For Each Loop(Array:Array of Object References=GetBoundObjects!(ObjectBinding:Movie Scene Object Binding ID Structure=@n4) -> (ReturnValue:Array of Object References)) -> (Array Element:Object Reference)
   n5 = Get binding_id() -> (binding_id:Movie Scene Object Binding ID Structure)
   n6 = "Get (a copy)"(Array:Array of Object References=GetBoundObjects(ObjectBinding=@n5)) -> (Output:Object Reference)
   n7 = Cast To FortPlayerMannequin!(execute=@n13.then, Object=@n6) -> (then, CastFailed, AsFort Player Mannequin:Fort Player Mannequin Object Reference) {is=title K2Node_DynamicCast}
   n8 = GetAllPartSkeletalMeshComponents!(execute, self:Fort Player Mannequin Object Reference=@n7) -> (then, SkeletalMeshComponents:Array of Skeletal Mesh Component Object References)
   > n9 = For Each Loop(Array:Array of Skeletal Mesh Component Object References=@n8) -> (Array Element:Skeletal Mesh Component Object Reference)
   > n10 = Create Widget!(
    execute:Exec
    WorldContextObject:Object Reference
    Class:User Widget Class Reference=&/FortniteAnimTools_ConfigureControlRigRowItem.FortniteAnimTools_ConfigureControlRigRowItem_C
    OwningPlayer:Player Controller Object Reference
    component_name:Text="To Text (String)"(InString=GetObjectName(Object=@n9))
    binding_id:Movie Scene Object Binding ID Structure
   ) -> (then:Exec, ReturnValue:Fortnite Anim Tools Configure Control Rig Row Item Object Reference)
   n11 = Get control_rigs_table_grid_layout() -> (control_rigs_table_grid_layout:Grid Panel Object Reference)
   > AddChildToGrid!(execute, self:Grid Panel Object Reference=@n11, Content:Widget Object Reference=@n10, InRow:int=int + int(A=@n9.Array Index), InColumn:int=0) -> (then, ReturnValue:Grid Slot Object Reference)
   n12 = GetCurrentLevelSequence!(execute) -> (then, ReturnValue:Level Sequence Object Reference)
   "On Clicked (mannequin_settings_close_btn)"() {is=title K2Node_ComponentBoundEvent}
   > RemoveFromParent()
   Event Construct()
   > n13 = SetObject(self=Get mannequin_settings_detail_view() -> (mannequin_settings_detail_view:Details View Object Reference), NewObject=@n6)
   n14 = Create Widget(
    execute=@n7.then
    Class=&FortniteAnimTools_ConfigureControlRigRowItem_C
    component_name="LOCTEXT(\"Actor\")"
    binding_id=@n5
   )
   > AddChildToGrid(self=@n11, Content=@n14, InRow=0, InColumn=0)
   Event PreConstruct()
   Event Tick()
  ]}]
variables [
 binding_id {type=MovieSceneObjectBindingID, default {guid=#1, resolveParentIndex=-1, sequenceId=0}}
 mannequin_obj {type=Object, default=null}
]
widgets {
 info {rootWidgetClass=&Border, widgetCount=17}
 widgets [
  Border_2 {
   widgetClassPath=&Border
   widget=Border_2
   properties {horizontalAlignment=HAlign_Center, brushColor=<rgba 0.0073 0.0073 0.0073 0.9>}
  }
  SizeBox_2 {
   widgetClassPath=&SizeBox
   parent=Border_2
   widget=SizeBox_2
   properties {bOverride_MinDesiredWidth=true, minDesiredWidth=600}
   slot {class=BorderSlot, properties {horizontalAlignment=HAlign_Center}}
  }
  mannequin_settings_vertical_layout {
   widgetClassPath=&VerticalBox
   bIsVariable=true
   parent=SizeBox_2
   widget=mannequin_settings_vertical_layout
   properties {clipping=ClipToBounds, visibility=Visible}
   slot {class=SizeBoxSlot, properties {verticalAlignment=VAlign_Center}}
  }
  Border_3 {
   widgetClassPath=&Border
   parent=mannequin_settings_vertical_layout
   widget=Border_3
   properties {
    background {
     bIsDynamicallyLoaded=false
     drawAs=Box
     imageType=NoImage
     mirroring=NoMirror
     resourceName=None
     resourceObject=null
     tiling=NoTile
     imageSize=<32 32>
     margin=<ltrb 0 0 0 0>
     outlineSettings {
      bUseBrushTransparency=false
      roundingType=HalfHeightRadius
      width=0
      color {colorUseRule=UseColor_Specified, specifiedColor=<rgba 0 0 0 0>}
      cornerRadii=<0 0 0 0>
     }
     tintColor {colorUseRule=UseColor_Specified, specifiedColor=<rgba 1 1 1 1>}
     uVRegion {bIsValid=false, max=<0 0>, min=<0 0>}
    }
    brushColor=<rgba 0.0885 0.0885 0.0885 1>
    padding=<ltrb 5 5 5 5>
   }
   slot {class=VerticalBoxSlot}
  }
  HorizontalBox_14 {
   widgetClassPath=&HorizontalBox
   parent=Border_3
   widget=HorizontalBox_14
   slot {class=BorderSlot, properties {padding=<ltrb 5 5 5 5>}}
  }
  TextBlock_2 {
   widgetClassPath=&TextBlock
   parent=HorizontalBox_14
   widget=TextBlock_2
   properties {
    clipping=ClipToBounds
    text=Mannequin Config
    font {
     bForceMonospaced=false
     bMaterialIsStencil=false
     fontMaterial=null
     fontObject=&Roboto
     letterSpacing=0
     monospacedWidth=1
     size=12
     skewAmount=0
     typefaceFontName=Regular
     outlineSettings {
      bApplyOutlineToDropShadows=false
      bMiteredCorners=false
      bSeparateFillAlpha=false
      outlineMaterial=null
      outlineSize=0
      outlineColor=<rgba 0 0 0 1>
     }
    }
   }
   slot {
    class=HorizontalBoxSlot
    properties {horizontalAlignment=HAlign_Left, verticalAlignment=VAlign_Center, padding=<ltrb 5 0 0 0>, size {sizeRule=Fill, value=1}}
   }
  }
  mannequin_settings_close_btn {
   widgetClassPath=&EditorUtilityButton
   parent=HorizontalBox_14
   widget=mannequin_settings_close_btn
   slot {class=HorizontalBoxSlot, properties {horizontalAlignment=HAlign_Right, verticalAlignment=VAlign_Top, padding=<ltrb 5 5 5 5>}}
  }
  TextBlock_10 {
   widgetClassPath=&TextBlock
   parent=mannequin_settings_close_btn
   widget=TextBlock_10
   properties {
    text=Close
    font {
     bForceMonospaced=false
     bMaterialIsStencil=false
     fontMaterial=null
     fontObject=&Roboto
     letterSpacing=0
     monospacedWidth=1
     size=9
     skewAmount=0
     typefaceFontName=Bold
     outlineSettings {
      bApplyOutlineToDropShadows=false
      bMiteredCorners=false
      bSeparateFillAlpha=false
      outlineMaterial=null
      outlineSize=0
      outlineColor=<rgba 0 0 0 1>
     }
    }
   }
   slot {class=ButtonSlot, properties {horizontalAlignment=HAlign_Fill, verticalAlignment=VAlign_Fill, padding=<ltrb 5 5 5 5>}}
  }
  Border_1 {
   widgetClassPath=&Border
   parent=mannequin_settings_vertical_layout
   widget=Border_1
   properties {
    background {
     bIsDynamicallyLoaded=false
     drawAs=RoundedBox
     imageType=NoImage
     mirroring=NoMirror
     resourceName=None
     resourceObject=null
     tiling=NoTile
     imageSize=<32 32>
     margin=<ltrb 0 0 0 0>
     outlineSettings {
      bUseBrushTransparency=false
      roundingType=FixedRadius
      width=0
      color {colorUseRule=UseColor_Specified, specifiedColor=<rgba 0 0 0 0>}
      cornerRadii=<0 0 5 5>
     }
     tintColor {colorUseRule=UseColor_Specified, specifiedColor=<rgba 1 1 1 1>}
     uVRegion {bIsValid=false, max=<0 0>, min=<0 0>}
    }
    brushColor=<rgba 0.0885 0.0885 0.0885 1>
    padding=<ltrb 4 2 4 10>
   }
   slot {class=VerticalBoxSlot, properties {size {sizeRule=Fill, value=1}}}
  }
  EditorUtilityScrollBox {
   widgetClassPath=&EditorUtilityScrollBox
   parent=Border_1
   widget=EditorUtilityScrollBox
   properties {
    orientation=Orient_Horizontal
    widgetStyle {
     barThickness=8
     bottomShadowBrush {
      bIsDynamicallyLoaded=false
      drawAs=NoDrawType
      imageType=FullColor
      mirroring=NoMirror
      resourceName=../../../Engine/Content/Slate/Common/ScrollBoxShadowBottom.png
      resourceObject=null
      tiling=NoTile
      imageSize=<16 8>
      margin=<ltrb 0.5 0 0.5 1>
      outlineSettings {
       bUseBrushTransparency=false
       roundingType=HalfHeightRadius
       width=0
       color {colorUseRule=UseColor_Specified, specifiedColor=<rgba 0 0 0 0>}
       cornerRadii=<0 0 0 0>
      }
      tintColor {colorUseRule=UseColor_Specified, specifiedColor=<rgba 1 1 1 1>}
      uVRegion {bIsValid=false, max=<0 0>, min=<0 0>}
     }
     horizontalScrolledContentPadding=<ltrb 0 0 1 0>
     leftShadowBrush {
      bIsDynamicallyLoaded=false
      drawAs=NoDrawType
      imageType=FullColor
      mirroring=NoMirror
      resourceName=../../../Engine/Content/Slate/Common/ScrollBoxShadowLeft.png
      resourceObject=null
      tiling=NoTile
      imageSize=<8 16>
      margin=<ltrb 1 0.5 0 0.5>
      outlineSettings {
       bUseBrushTransparency=false
       roundingType=HalfHeightRadius
       width=0
       color {colorUseRule=UseColor_Specified, specifiedColor=<rgba 0 0 0 0>}
       cornerRadii=<0 0 0 0>
      }
      tintColor {colorUseRule=UseColor_Specified, specifiedColor=<rgba 1 1 1 1>}
      uVRegion {bIsValid=false, max=<0 0>, min=<0 0>}
     }
     rightShadowBrush {
      bIsDynamicallyLoaded=false
      drawAs=NoDrawType
      imageType=FullColor
      mirroring=NoMirror
      resourceName=../../../Engine/Content/Slate/Common/ScrollBoxShadowRight.png
      resourceObject=null
      tiling=NoTile
      imageSize=<8 16>
      margin=<ltrb 0 0.5 1 0.5>
      outlineSettings {
       bUseBrushTransparency=false
       roundingType=HalfHeightRadius
       width=0
       color {colorUseRule=UseColor_Specified, specifiedColor=<rgba 0 0 0 0>}
       cornerRadii=<0 0 0 0>
      }
      tintColor {colorUseRule=UseColor_Specified, specifiedColor=<rgba 1 1 1 1>}
      uVRegion {bIsValid=false, max=<0 0>, min=<0 0>}
     }
     topShadowBrush {
      bIsDynamicallyLoaded=false
      drawAs=NoDrawType
      imageType=FullColor
      mirroring=NoMirror
      resourceName=../../../Engine/Content/Slate/Common/ScrollBoxShadowTop.png
      resourceObject=null
      tiling=NoTile
      imageSize=<16 8>
      margin=<ltrb 0.5 1 0.5 0>
      outlineSettings {
       bUseBrushTransparency=false
       roundingType=HalfHeightRadius
       width=0
       color {colorUseRule=UseColor_Specified, specifiedColor=<rgba 0 0 0 0>}
       cornerRadii=<0 0 0 0>
      }
      tintColor {colorUseRule=UseColor_Specified, specifiedColor=<rgba 1 1 1 1>}
      uVRegion {bIsValid=false, max=<0 0>, min=<0 0>}
     }
     verticalScrolledContentPadding=<ltrb 0 0 0 1>
    }
   }
   slot {class=BorderSlot, properties {padding=<ltrb 4 2 4 10>}}
  }
  mannequin_settings_vertical_scroll {
   widgetClassPath=&EditorUtilityScrollBox
   bIsVariable=true
   parent=EditorUtilityScrollBox
   widget=mannequin_settings_vertical_scroll
   slot {class=ScrollBoxSlot, properties {size {sizeRule=Fill, value=1}}}
  }
  Border_0 {
   widgetClassPath=&Border
   parent=mannequin_settings_vertical_scroll
   widget=Border_0
   properties {brushColor=<rgba 0.0104 0.0104 0.0104 1>, padding=<ltrb 5 5 5 5>}
   slot {class=ScrollBoxSlot}
  }
  TextBlock_5 {
   widgetClassPath=&TextBlock
   parent=Border_0
   widget=TextBlock_5
   properties {
    text=Variant / Loadout:
    font {
     bForceMonospaced=false
     bMaterialIsStencil=false
     fontMaterial=null
     fontObject=&Roboto
     letterSpacing=0
     monospacedWidth=1
     size=9
     skewAmount=0
     typefaceFontName=Bold
     outlineSettings {
      bApplyOutlineToDropShadows=false
      bMiteredCorners=false
      bSeparateFillAlpha=false
      outlineMaterial=null
      outlineSize=0
      outlineColor=<rgba 0 0 0 1>
     }
    }
   }
   slot {class=BorderSlot, properties {padding=<ltrb 5 5 5 5>}}
  }
  mannequin_settings_detail_view {
   widgetClassPath=&DetailsView
   parent=mannequin_settings_vertical_scroll
   widget=mannequin_settings_detail_view
   properties {categoriesToShow [Weapon], propertiesToShow [OverrideVariants, AthenaBackBling, CharacterPartList]}
   slot {class=ScrollBoxSlot}
  }
  Border {
   widgetClassPath=&Border
   parent=mannequin_settings_vertical_scroll
   widget=Border
   properties {brushColor=<rgba 0.0104 0.0104 0.0104 1>, padding=<ltrb 5 5 5 5>}
   slot {class=ScrollBoxSlot}
  }
  TextBlock {
   widgetClassPath=&TextBlock
   parent=Border
   widget=TextBlock
   properties {
    text=Control Rigs
    font {
     bForceMonospaced=false
     bMaterialIsStencil=false
     fontMaterial=null
     fontObject=&Roboto
     letterSpacing=0
     monospacedWidth=1
     size=9
     skewAmount=0
     typefaceFontName=Bold
     outlineSettings {
      bApplyOutlineToDropShadows=false
      bMiteredCorners=false
      bSeparateFillAlpha=false
      outlineMaterial=null
      outlineSize=0
      outlineColor=<rgba 0 0 0 1>
     }
    }
   }
   slot {class=BorderSlot, properties {padding=<ltrb 5 5 5 5>}}
  }
  control_rigs_table_grid_layout {
   widgetClassPath=&GridPanel
   bIsVariable=true
   parent=mannequin_settings_vertical_scroll
   widget=control_rigs_table_grid_layout
   properties {columnFill [1]}
   slot {class=ScrollBoxSlot, properties {size {sizeRule=Fill, value=1}}}
  }
 ]
}
