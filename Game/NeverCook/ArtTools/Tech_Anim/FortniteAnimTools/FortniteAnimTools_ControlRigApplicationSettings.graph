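# Editor utility widget in the Fortnite Anim Tools that serves as a dialog for editing Control Rig application settings.
# On construct, a DetailsView binds to the widget itself, showing its Control Rig Settings category, and the defaults button takes the current workflow's name.
# That button copies the settings from the soft-referenced workflow.
# OK broadcasts the settings through the ControlRigSettingsOkClicked dispatcher, then closes; Cancel just closes.
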EditorUtilityWidgetBlueprint /Game/NeverCook/ArtTools/Tech_Anim/FortniteAnimTools/FortniteAnimTools_ControlRigApplicationSettings
graphs [
 ControlRigSettingsOkClicked {nodes [ControlRigSettingsOkClicked!() -> (then, control_rig_settings:Fortnite Anim Tools Control Rig Settings Structure) {is=title K2Node_FunctionEntry}]}
 EventGraph {nodes [
   "On Clicked (Ok_btn)"() {is=title K2Node_ComponentBoundEvent}
   > n1 = Call Control Rig Settings Ok Clicked!(
    execute:Exec
    self:Fortnite Anim Tools Control Rig Application Settings Object Reference
    control_rig_settings:Fortnite Anim Tools Control Rig Settings Structure=Get control_rig_settings() -> (control_rig_settings:Fortnite Anim Tools Control Rig Settings Structure)
   ) -> (then:Exec) {is=title K2Node_CallDelegate}
   RemoveFromParent(execute=[@n1.then, @n2.then])
   n2 = "On Clicked (Cancel_btn)"() {is=title K2Node_ComponentBoundEvent}
   "On Clicked (set_to_workflow_defaults_btn)"() {is=title K2Node_ComponentBoundEvent}
   n3 = Get current_workflow() -> (current_workflow:Fortnite Anim Tools Workflow Soft Object Reference)
   > Set control_rig_settings(control_rig_settings:Fortnite Anim Tools Control Rig Settings Structure=Get control_rig_settings!(self:Fortnite Anim Tools Workflow Object Reference=Resolve Soft Reference!(Input:Fortnite Anim Tools Workflow Soft Object Reference=@n3) -> (Output:Fortnite Anim Tools Workflow Object Reference)) -> (control_rig_settings)) -> (Output_Get:Fortnite Anim Tools Control Rig Settings Structure)
   Event Construct()
   > SetObject(self=Get settings_view() -> (settings_view:Details View Object Reference), NewObject=Self-Reference())
   n4 = Get name!(self:Fortnite Anim Tools Workflow Object Reference=Resolve Soft Reference(Input=Get current_workflow())) -> (name:String)
   > "SetText (Text)"(
    self=Get set_to_workflow_defaults_txt() -> (set_to_workflow_defaults_txt:Text Object Reference)
    InText="To Text (String)"(InString=Append!(A="Set To  ", B=@n4, C:String=" Workflow Defaults") -> (ReturnValue))
   )
   Event PreConstruct()
   Event Tick()
  ]}
]
variables [
 control_rig_settings {
  type=FortniteAnimTools_ControlRigSettings
  default {
   bake_animation_settings {reduce_keys=false, reduce_keys_tolerance=0, remove_twist_control_keys=false, reset_controls=false}
   biped_rig_settings {apply=false, layered_mode=false, default_poses []}
   extraJoint_rig_settings {apply=false, layered_mode=false}
   face_rig_settings {apply=false, apply_tweakers=false, layered_mode=false, default_poses []}
  }
 }
 ControlRigSettingsOkClicked {default=()}
 current_workflow {type=FortniteAnimToolsWorkflow_C, default=null}
]
widgets {
 info {rootWidgetClass=&Border, widgetCount=11}
 widgets [
  Border_0 {
   widgetClassPath=&Border
   widget=Border_0
   properties {horizontalAlignment=HAlign_Center, verticalAlignment=VAlign_Center, brushColor=<rgba 0.0073 0.0073 0.0073 0.9>}
  }
  SizeBox_0 {
   widgetClassPath=&SizeBox
   parent=Border_0
   widget=SizeBox_0
   properties {bOverride_MinDesiredWidth=true, minDesiredWidth=600}
   slot {class=BorderSlot, properties {horizontalAlignment=HAlign_Center, verticalAlignment=VAlign_Center}}
  }
  VerticalBox_40 {widgetClassPath=&VerticalBox, parent=SizeBox_0, widget=VerticalBox_40, slot {class=SizeBoxSlot}}
  settings_view {
   widgetClassPath=&DetailsView
   parent=VerticalBox_40
   widget=settings_view
   properties {viewIdentifier=Control Rig Settings Panel, categoriesToShow [Control Rig Settings]}
   slot {class=VerticalBoxSlot, properties {size {sizeRule=Fill, value=1}}}
  }
  set_to_workflow_defaults_btn {
   widgetClassPath=&EditorUtilityButton
   parent=VerticalBox_40
   widget=set_to_workflow_defaults_btn
   slot {class=VerticalBoxSlot}
  }
  set_to_workflow_defaults_txt {
   widgetClassPath=&TextBlock
   bIsVariable=true
   parent=set_to_workflow_defaults_btn
   widget=set_to_workflow_defaults_txt
   properties {
    text=Set To Workflow Defaults
    font {
     bForceMonospaced=false
     bMaterialIsStencil=false
     fontMaterial=null
     fontObject=&Roboto
     letterSpacing=0
     monospacedWidth=1
     size=12
     skewAmount=0
     typefaceFontName=Bold
     outlineSettings {
      bApplyOutlineToDropShadows=false
      bMiteredCorners=false
      bSeparateFillAlpha=false
      outlineMaterial=null
      outlineSize=0
      outlineColor=<rgba 0 0 0 1>
     }
    }
   }
   slot {class=ButtonSlot}
  }
  HorizontalBox_1 {
   widgetClassPath=&HorizontalBox
   parent=VerticalBox_40
   widget=HorizontalBox_1
   slot {class=VerticalBoxSlot, properties {verticalAlignment=VAlign_Center}}
  }
  Ok_btn {
   widgetClassPath=&EditorUtilityButton
   parent=HorizontalBox_1
   widget=Ok_btn
   slot {class=HorizontalBoxSlot, properties {size {sizeRule=Fill, value=0.3}}}
  }
  TextBlock_81 {
   widgetClassPath=&TextBlock
   parent=Ok_btn
   widget=TextBlock_81
   properties {
    text=OK
    font {
     bForceMonospaced=false
     bMaterialIsStencil=false
     fontMaterial=null
     fontObject=&Roboto
     letterSpacing=0
     monospacedWidth=1
     size=12
     skewAmount=0
     typefaceFontName=Bold
     outlineSettings {
      bApplyOutlineToDropShadows=false
      bMiteredCorners=false
      bSeparateFillAlpha=false
      outlineMaterial=null
      outlineSize=0
      outlineColor=<rgba 0 0 0 1>
     }
    }
   }
   slot {class=ButtonSlot, properties {padding=<ltrb 4 4 4 2>}}
  }
  Cancel_btn {
   widgetClassPath=&EditorUtilityButton
   parent=HorizontalBox_1
   widget=Cancel_btn
   slot {class=HorizontalBoxSlot, properties {size {sizeRule=Fill, value=0.3}}}
  }
  TextBlock {
   widgetClassPath=&TextBlock
   parent=Cancel_btn
   widget=TextBlock
   properties {
    text=Cancel
    font {
     bForceMonospaced=false
     bMaterialIsStencil=false
     fontMaterial=null
     fontObject=&Roboto
     letterSpacing=0
     monospacedWidth=1
     size=12
     skewAmount=0
     typefaceFontName=Bold
     outlineSettings {
      bApplyOutlineToDropShadows=false
      bMiteredCorners=false
      bSeparateFillAlpha=false
      outlineMaterial=null
      outlineSize=0
      outlineColor=<rgba 0 0 0 1>
     }
    }
   }
   slot {class=ButtonSlot, properties {padding=<ltrb 4 4 4 2>}}
  }
 ]
}
