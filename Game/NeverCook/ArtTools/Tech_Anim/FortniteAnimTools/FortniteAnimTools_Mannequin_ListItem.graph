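# Editor utility list row in Fortnite Anim Tools for one Sequencer-bound FortPlayerMannequin.
# A list data object supplies the binding, driving the editable actor label and a self-targeted details view whose character edits set AthenaCharacter and OverrideVariants.
# Apply and configure buttons broadcast dispatchers for external handling.
# Delete removes Control Rig tracks from the binding and Python-validated children after confirmation; another button clears twist-bone keys through inline Python.
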EditorUtilityWidgetBlueprint /Game/NeverCook/ArtTools/Tech_Anim/FortniteAnimTools/FortniteAnimTools_Mannequin_ListItem
graphs [
 ActorNameEditing {nodes [ActorNameEditing() {is=title K2Node_FunctionEntry}]}
 ActorNameEditingFinished {nodes [ActorNameEditingFinished() {is=title K2Node_FunctionEntry}]}
 EventGraph {nodes [
   n1 = GetCurrentLevelSequence!(execute) -> (then, ReturnValue:Level Sequence Object Reference)
   n2 = GetFocusedLevelSequence!(execute=@n37.then) -> (then, ReturnValue:Level Sequence Object Reference)
   n3 = Get object_holder() -> (object_holder:Fortnite Anim Tools Mannequin List Item Data Soft Object Reference)
   n4 = ResolveBindingID!(
    execute=[@n1.then, @n2.then]
    RootSequence:Movie Scene Sequence Object Reference=@n2
    InObjectBindingID:Movie Scene Object Binding ID Structure=Get binding_id!(self:Fortnite Anim Tools Mannequin List Item Data Object Reference=Resolve Soft Reference!(Input:Fortnite Anim Tools Mannequin List Item Data Soft Object Reference=@n3) -> (Output:Fortnite Anim Tools Mannequin List Item Data Object Reference)) -> (binding_id:Movie Scene Object Binding ID Structure)
   ) -> (then, ReturnValue:Movie Scene Binding Proxy Structure)
   > Call On Apply Control Rigs Clicked!(execute:Exec, self:Fortnite Anim Tools Mannequin List Item Object Reference, actor_binding:Movie Scene Binding Proxy Structure=@n4) -> (then:Exec) {is=title K2Node_CallDelegate}
   n5 = GetCurrentLevelSequence()
   n6 = GetFocusedLevelSequence(execute=@n36.then)
   n7 = Get object_holder()
   n8 = ResolveBindingID(execute=[@n6.then, @n5.then], RootSequence=@n6, InObjectBindingID=Get binding_id(self=Resolve Soft Reference(Input=@n7)))
   > n9 = GetDisplayName!(execute:Exec, InBinding:"Movie Scene Binding Proxy Structure (by ref)"=@n8) -> (then:Exec, ReturnValue:Text)
   n10 = "To String (Text)"(InText=@n9)
   > n11 = Show Message Dialog(
    Title=MakeLiteralText(Value="LOCTEXT(\"Delete Control Rigs?\")")
    Message="To Text (String)"(InString=Append!(A="Are you sure you want to delete all control rigs on ", B=@n10, C:String=" ?") -> (ReturnValue))
    MessageType=OkCancel
    DefaultValue=No
    MessageCategory=Info
   )
   > Branch(Condition="Equal (Enum)"(A:EAppReturnType Enum=@n11, B:EAppReturnType Enum=Ok))
   > n12 = FindTracksByType!(
    execute
    InBinding:"Movie Scene Binding Proxy Structure (by ref)"=@n8
    TrackType:Movie Scene Track Class Reference=/Script/ControlRig.MovieSceneControlRigParameterTrack
   ) -> (then, ReturnValue:Array of Movie Scene Control Rig Parameter Track Object References)
   > n13 = For Each Loop(Array:Array of Movie Scene Control Rig Parameter Track Object References=@n12) -> (Array Element:Movie Scene Control Rig Parameter Track Object Reference)
   n14 = GetChildPossessables!(execute=@n13.Completed, InBinding:"Movie Scene Binding Proxy Structure (by ref)"=@n8) -> (then, ReturnValue:Array of Movie Scene Binding Proxy Structures)
   > n15 = For Each Loop(Array:Array of Movie Scene Binding Proxy Structures=@n14) -> (Array Element:Movie Scene Binding Proxy Structure)
   > n16 = Execute Python Script!(
    execute
    PythonScript="import importlib\r\nimport unreal\r\nimport TechAnim.sequencer.ctrlrig_utils as ctrlrig_utils\r\n\r\nimportlib.reload(ctrlrig_utils)\r\n\r\nis_valid_component = ctrlrig_utils.is_valid_component(component_name)"
    component_name:Movie Scene Binding Proxy Structure=@n15
   ) -> (then, ReturnValue, is_valid_component:bool=~) {inputs [component_name], outputs [is_valid_component]}
   > Branch(Condition=@n16.is_valid_component)
   > n17 = FindTracksByType(InBinding=@n15, TrackType=&MovieSceneControlRigParameterTrack)
   > n18 = For Each Loop(Array:Array of Movie Scene Control Rig Parameter Track Object References=@n17) -> (Array Element:Movie Scene Control Rig Parameter Track Object Reference)
   > RemoveTrack!(execute, InBinding:"Movie Scene Binding Proxy Structure (by ref)"=@n15, TrackToRemove:Movie Scene Track Object Reference=@n18) -> (then)
   RemoveTrack(execute=@n13.LoopBody, InBinding=@n8, TrackToRemove=@n13)
   n19 = GetCurrentLevelSequence()
   n20 = GetFocusedLevelSequence(execute=@n38.then)
   n21 = Get object_holder()
   n22 = ResolveBindingID(execute=[@n20.then, @n19.then], RootSequence=@n20, InObjectBindingID=Get binding_id(self=Resolve Soft Reference(Input=@n21)))
   > Execute Python Script!(
    execute
    PythonScript="import importlib\r\nimport unreal\r\nimport TechAnim.sequencer.ctrlrig_utils as ctrlrig_utils\r\n\r\nimportlib.reload(ctrlrig_utils)\r\n\r\nfor binding in bindings:\r\n    control_rig_tracks = binding.find_tracks_by_type(unreal.MovieSceneControlRigParameterTrack)\r\n    for control_rig in control_rig_tracks:\r\n        for control_rig_track in control_rig_tracks:\r\n            twist_channels = ctrlrig_utils.get_twist_bone_control_channels(control_rig_track)\r\n            for channel in twist_channels:\r\n                ctrlrig_utils.remove_keys_and_reset_channel(channel)"
    bindings:Array of Movie Scene Binding Proxy Structures=Make Array("[0]":Movie Scene Binding Proxy Structure=@n22) -> (Array:Array of Movie Scene Binding Proxy Structures)
   ) -> (then, ReturnValue) {inputs [bindings]}
   n23 = "On Property Changed (Mannequin_Details)"!() -> (OutputDelegate, then, PropertyName:Name=None) {is=title K2Node_ComponentBoundEvent}
   > Branch(Condition="Equal (Name)"(A=@n23.PropertyName, B=cid_obj))
   > n24 = LoadAsset_Blocking(Asset=Get cid_obj() -> (cid_obj:Athena Character Item Definition Soft Object Reference))
   n25 = Resolve Soft Reference(Input=Get object_holder())
   n26 = "Get (a copy)"(Array:Array of Object References=GetBoundObjects!(ObjectBinding:Movie Scene Object Binding ID Structure=Get binding_id(self=@n25)) -> (ReturnValue:Array of Object References)) -> (Output:Object Reference)
   > n27 = Is Valid(InputObject=@n26)
   PrintString(execute=@n27.Is Not Valid, InString=binding is not valid)
   n28 = Cast To FortPlayerMannequin!(execute=@n27.Is Valid, Object=@n26) -> (then, CastFailed, AsFort Player Mannequin:Fort Player Mannequin Object Reference) {is=title K2Node_DynamicCast}
   > n29 = Cast To AthenaCharacterItemDefinition!(execute, Object=@n24) -> (then, CastFailed, AsAthena Character Item Definition:Athena Character Item Definition Object Reference) {is=title K2Node_DynamicCast}
   > SetEditorProperty(Object=@n28, PropertyName=AthenaCharacter, PropertyValue:Athena Character Item Definition Object Reference=@n29)
   > n30 = MakeVariantChannelInfo!(
    execute
    ItemDefinition:Athena Cosmetic Item Definition Object Reference=@n29
    Channel:Gameplay Tag Structure
    ActiveVariant:Gameplay Tag Structure
    CustomData:String
   ) -> (then, ReturnValue:Mcp Variant Channel Info Structure)
   > SetEditorProperty(
    Object=@n28
    PropertyName=OverrideVariants
    PropertyValue:Array of Mcp Variant Channel Info Structures=Make Array("[0]":Mcp Variant Channel Info Structure=@n30) -> (Array:Array of Mcp Variant Channel Info Structures)
   )
   PrintString(execute=@n29.CastFailed, InString=Cast To AthenaItemDefinitionFailed)
   PrintString(execute=@n28.CastFailed, InString=Cast To FortPlayerMannequin Failed)
   "On Clicked (configure_btn)"() {is=title K2Node_ComponentBoundEvent}
   n31 = Get binding_id(self=Resolve Soft Reference(Input=Get object_holder()))
   n32 = GetBoundObjects(ObjectBinding=@n31)
   > Call On Configure Btn Clicked!(
    execute:Exec
    self:Fortnite Anim Tools Mannequin List Item Object Reference
    mannequin_obj:Soft Object Reference=Make Soft Reference!(Input:Object Reference="Get (a copy)"(Array=@n32)) -> (Output:Soft Object Reference)
    binding_id:Movie Scene Object Binding ID Structure=@n31
   ) -> (then:Exec) {is=title K2Node_CallDelegate}
   n33 = "OnTextCommitted (Editable Text) (MannequinName_TxtBx)"!() -> (OutputDelegate, then, Text:"Text (by ref)", CommitMethod:ETextCommit Enum=Default) {is=title K2Node_ComponentBoundEvent}
   n34 = Get binding_id(self=Resolve Soft Reference(Input=Get object_holder()))
   > n35 = Cast To FortPlayerMannequin(Object="Get (a copy)"(Array=GetBoundObjects(ObjectBinding=@n34)))
   > SetActorLabel(self=@n35, NewActorLabel="To String (Text)"(InText=@n33.Text), bMarkDirty=true)
   > Call Actor Name Editing Finished!(execute:Exec, self:Fortnite Anim Tools Mannequin List Item Object Reference) -> (then:Exec) {is=title K2Node_CallDelegate}
   n36 = "On Clicked (delete_control_rigs_btn)"() {is=title K2Node_ComponentBoundEvent}
   n37 = "On Clicked (apply_control_rigs_btn)"() {is=title K2Node_ComponentBoundEvent}
   "OnTextChanged (Editable Text) (MannequinName_TxtBx)"!() -> (OutputDelegate, then, Text:"Text (by ref)") {is=title K2Node_ComponentBoundEvent}
   > Call Actor Name Editing!(execute:Exec, self:Fortnite Anim Tools Mannequin List Item Object Reference) -> (then:Exec) {is=title K2Node_CallDelegate}
   n38 = "On Clicked (remove_twist_keys_btn)"() {is=title K2Node_ComponentBoundEvent}
   n39 = Event OnListItemObjectSet()
   > n40 = Cast To FortniteAnimTools_MannequinListItemData!(execute, Object=@n39.ListItemObject) -> (then, CastFailed, AsFortnite Anim Tools Mannequin List Item Data:Fortnite Anim Tools Mannequin List Item Data Object Reference) {is=title K2Node_DynamicCast}
   n41 = "Get (a copy)"(Array=GetBoundObjects(ObjectBinding=Get binding_id(self=@n40)))
   > n42 = Is Valid(InputObject=@n41)
   PrintString(execute=@n42.Is Not Valid, InString=binding not valid)
   n43 = Cast To FortPlayerMannequin(execute=@n42.Is Valid, Object=@n41)
   > Set cid_obj(cid_obj:Athena Character Item Definition Soft Object Reference=Make Soft Reference(Input:Athena Character Item Definition Object Reference=GetAthenaCharacterItemDefinition!(self:Fort Player Mannequin Object Reference=@n43) -> (ReturnValue:Athena Character Item Definition Object Reference)) -> (Output:Athena Character Item Definition Soft Object Reference)) -> (Output_Get:Athena Character Item Definition Soft Object Reference)
   > Set object_holder(object_holder:Fortnite Anim Tools Mannequin List Item Data Soft Object Reference=Make Soft Reference(Input:Fortnite Anim Tools Mannequin List Item Data Object Reference=@n40) -> (Output:Fortnite Anim Tools Mannequin List Item Data Soft Object Reference)) -> (Output_Get:Fortnite Anim Tools Mannequin List Item Data Soft Object Reference)
   > Set Text(
    Text="To Text (String)"(InString=GetActorLabel(self=@n43, bCreateIfNone=true))
    self:Editable Text Object Reference=Get MannequinName_TxtBx() -> (MannequinName_TxtBx:Editor Utility Editable Text Object Reference)
   )
   > SetObject(self=Get Mannequin_Details() -> (Mannequin_Details:Details View Object Reference), NewObject=Self-Reference())
   PrintString(execute=@n43.CastFailed, InString=Failed cast to FortPlayerMannequin)
   n44 = Event On Item Selection Changed!() -> (OutputDelegate, then, bIsSelected:bool) {is=title K2Node_Event}
   > SetBrushColor(
    self=Get Border() -> (Border:Border Object Reference)
    InBrushColor=SelectColor(A=(R=0,G=0.162,B=0.7454,A=0.577), B=(R=0.0578,G=0.0578,B=0.0578,A=1), bPickA=@n44.bIsSelected)
   )
   Event PreConstruct()
   Event Construct()
   Event Tick()
  ]}
 On Configure Btn Clicked {nodes [On Configure Btn Clicked!() -> (then, mannequin_obj:Soft Object Reference, binding_id:Movie Scene Object Binding ID Structure) {is=title K2Node_FunctionEntry}]}
 OnApplyControlRigsClicked {nodes [OnApplyControlRigsClicked!() -> (then, actor_binding:Movie Scene Binding Proxy Structure) {is=title K2Node_FunctionEntry}]}
]
variables [
 ActorNameEditing {default=()}
 ActorNameEditingFinished {default=()}
 AthenaCharacterProperty {type=string, default=AthenaCharacter}
 cid_obj {type=AthenaCharacterItemDefinition, default=null}
 object_holder {type=FortniteAnimTools_MannequinListItemData_C, default=null}
 On Configure Btn Clicked {default=()}
 OnApplyControlRigsClicked {default=()}
]
widgets {
 info {rootWidgetClass=&Border, widgetCount=21}
 widgets [
  Border {
   widgetClassPath=&Border
   bIsVariable=true
   widget=Border
   properties {
    background {
     bIsDynamicallyLoaded=false
     drawAs=RoundedBox
     imageType=NoImage
     mirroring=NoMirror
     resourceName=None
     resourceObject=null
     tiling=NoTile
     imageSize=<32 32>
     margin=<ltrb 0 0 0 0>
     outlineSettings {
      bUseBrushTransparency=false
      roundingType=FixedRadius
      width=0
      color {colorUseRule=UseColor_Specified, specifiedColor=<rgba 0 0 0 0>}
      cornerRadii=<5 5 5 5>
     }
     tintColor {colorUseRule=UseColor_Specified, specifiedColor=<rgba 1 1 1 1>}
     uVRegion {bIsValid=false, max=<0 0>, min=<0 0>}
    }
    brushColor=<rgba 0.0578 0.0578 0.0578 1>
    padding=<ltrb 10 2 2 2>
   }
  }
  VerticalBox_0 {widgetClassPath=&VerticalBox, parent=Border, widget=VerticalBox_0, slot {class=BorderSlot, properties {padding=<ltrb 10 2 2 2>}}}
  HorizontalBox_0 {widgetClassPath=&HorizontalBox, parent=VerticalBox_0, widget=HorizontalBox_0, slot {class=VerticalBoxSlot}}
  SizeBox_1 {
   widgetClassPath=&SizeBox
   parent=HorizontalBox_0
   widget=SizeBox_1
   properties {minDesiredWidth=100}
   slot {class=HorizontalBoxSlot, properties {verticalAlignment=VAlign_Center, size {sizeRule=Fill, value=0.3}}}
  }
  VerticalBox_1 {widgetClassPath=&VerticalBox, parent=SizeBox_1, widget=VerticalBox_1, slot {class=SizeBoxSlot}}
  MannequinName_TxtBx {
   widgetClassPath=/Script/Blutility.EditorUtilityEditableText
   bIsVariable=true
   parent=VerticalBox_1
   widget=MannequinName_TxtBx
   properties {
    text=mannequin
    widgetStyle {
     backgroundImageComposing {
      bIsDynamicallyLoaded=false
      drawAs=Border
      imageType=FullColor
      mirroring=NoMirror
      resourceName=../../../Engine/Content/Slate/Old/HyperlinkDotted.png
      resourceObject=null
      tiling=Both
      imageSize=<0 0>
      margin=<ltrb 0 0 0 0.1875>
      outlineSettings {
       bUseBrushTransparency=false
       roundingType=HalfHeightRadius
       width=0
       color {colorUseRule=UseColor_Specified, specifiedColor=<rgba 0 0 0 0>}
       cornerRadii=<0 0 0 0>
      }
      tintColor {colorUseRule=UseColor_Specified, specifiedColor=<rgba 1 1 1 1>}
      uVRegion {bIsValid=false, max=<0 0>, min=<0 0>}
     }
     backgroundImageSelected {
      bIsDynamicallyLoaded=false
      drawAs=Image
      imageType=NoImage
      mirroring=NoMirror
      resourceName=None
      resourceObject=null
      tiling=NoTile
      imageSize=<0 0>
      margin=<ltrb 0 0 0 0>
      outlineSettings {
       bUseBrushTransparency=false
       roundingType=HalfHeightRadius
       width=0
       color {colorUseRule=UseColor_Specified, specifiedColor=<rgba 0 0 0 0>}
       cornerRadii=<0 0 0 0>
      }
      tintColor {colorUseRule=UseColor_Specified, specifiedColor=<rgba 0 0.162 0.7454 1>}
      uVRegion {bIsValid=false, max=<0 0>, min=<0 0>}
     }
     caretImage {
      bIsDynamicallyLoaded=false
      drawAs=Image
      imageType=NoImage
      mirroring=NoMirror
      resourceName=None
      resourceObject=null
      tiling=NoTile
      imageSize=<0 0>
      margin=<ltrb 0 0 0 0>
      outlineSettings {
       bUseBrushTransparency=false
       roundingType=HalfHeightRadius
       width=0
       color {colorUseRule=UseColor_Specified, specifiedColor=<rgba 0 0 0 0>}
       cornerRadii=<0 0 0 0>
      }
      tintColor {colorUseRule=UseColor_Specified, specifiedColor=<rgba 1 1 1 1>}
      uVRegion {bIsValid=false, max=<0 0>, min=<0 0>}
     }
     colorAndOpacity {colorUseRule=UseColor_Foreground, specifiedColor=<rgba 1 0 1 1>}
     font {
      bForceMonospaced=false
      bMaterialIsStencil=false
      fontMaterial=null
      fontObject=&Roboto
      letterSpacing=0
      monospacedWidth=1
      size=10.5
      skewAmount=0
      typefaceFontName=Bold
      outlineSettings {
       bApplyOutlineToDropShadows=false
       bMiteredCorners=false
       bSeparateFillAlpha=false
       outlineMaterial=null
       outlineSize=0
       outlineColor=<rgba 0 0 0 1>
      }
     }
     hintColorAndOpacity {colorUseRule=UseColor_Foreground, specifiedColor=<rgba 1 0 1 1>}
     selectedBackgroundColor {colorUseRule=UseColor_Foreground, specifiedColor=<rgba 1 0 1 1>}
    }
   }
   slot {class=VerticalBoxSlot, properties {horizontalAlignment=HAlign_Left}}
  }
  SizeBox_2 {
   widgetClassPath=&SizeBox
   parent=VerticalBox_1
   widget=SizeBox_2
   properties {bOverride_MinDesiredWidth=true, minDesiredWidth=100}
   slot {class=VerticalBoxSlot, properties {horizontalAlignment=HAlign_Right, verticalAlignment=VAlign_Bottom, padding=<ltrb 0 15 0 0>}}
  }
  apply_control_rigs_btn {
   widgetClassPath=&EditorUtilityButton
   parent=SizeBox_2
   widget=apply_control_rigs_btn
   properties {
    clipping=ClipToBounds
    widgetStyle {
     clickedSlateSound {resourceObject=null}
     disabled {
      bIsDynamicallyLoaded=false
      drawAs=RoundedBox
      imageType=NoImage
      mirroring=NoMirror
      resourceName=None
      resourceObject=null
      tiling=NoTile
      imageSize=<0 0>
      margin=<ltrb 0 0 0 0>
      outlineSettings {
       bUseBrushTransparency=false
       roundingType=FixedRadius
       width=1
       color {colorUseRule=UseColor_ColorTable, specifiedColor=<rgba 0 0 0 0>}
       cornerRadii=<4 4 4 4>
      }
      tintColor {colorUseRule=UseColor_Specified, specifiedColor=<rgba 0.0395 0.0395 0.0395 1>}
      uVRegion {bIsValid=false, max=<0 0>, min=<0 0>}
     }
     disabledForeground {colorUseRule=UseColor_ColorTable, specifiedColor=<rgba 0 0 0 0>}
     hovered {
      bIsDynamicallyLoaded=false
      drawAs=RoundedBox
      imageType=NoImage
      mirroring=NoMirror
      resourceName=None
      resourceObject=null
      tiling=NoTile
      imageSize=<32 32>
      margin=<ltrb 0 0 0 0>
      outlineSettings {
       bUseBrushTransparency=false
       roundingType=FixedRadius
       width=1
       color {colorUseRule=UseColor_ColorTable, specifiedColor=<rgba 0 0 0 0>}
       cornerRadii=<4 4 4 4>
      }
      tintColor {colorUseRule=UseColor_Specified, specifiedColor=<rgba 0 0.6927 0.1824 1>}
      uVRegion {bIsValid=false, max=<0 0>, min=<0 0>}
     }
     hoveredForeground {colorUseRule=UseColor_ColorTable, specifiedColor=<rgba 0 0 0 0>}
     hoveredSlateSound {resourceObject=null}
     normal {
      bIsDynamicallyLoaded=false
      drawAs=RoundedBox
      imageType=NoImage
      mirroring=NoMirror
      resourceName=None
      resourceObject=null
      tiling=NoTile
      imageSize=<32 32>
      margin=<ltrb 0 0 0 0>
      outlineSettings {
       bUseBrushTransparency=false
       roundingType=FixedRadius
       width=1
       color {colorUseRule=UseColor_ColorTable, specifiedColor=<rgba 0 0 0 0>}
       cornerRadii=<4 4 4 4>
      }
      tintColor {colorUseRule=UseColor_Specified, specifiedColor=<rgba 0 0.4635 0.1159 1>}
      uVRegion {bIsValid=false, max=<0 0>, min=<0 0>}
     }
     normalForeground {colorUseRule=UseColor_ColorTable, specifiedColor=<rgba 0 0 0 0>}
     normalPadding=<ltrb 12 1.5 12 1.5>
     pressed {
      bIsDynamicallyLoaded=false
      drawAs=RoundedBox
      imageType=NoImage
      mirroring=NoMirror
      resourceName=None
      resourceObject=null
      tiling=NoTile
      imageSize=<32 32>
      margin=<ltrb 0 0 0 0>
      outlineSettings {
       bUseBrushTransparency=false
       roundingType=FixedRadius
       width=1
       color {colorUseRule=UseColor_ColorTable, specifiedColor=<rgba 0 0 0 0>}
       cornerRadii=<4 4 4 4>
      }
      tintColor {colorUseRule=UseColor_Specified, specifiedColor=<rgba 0.0284 0.0284 0.0284 1>}
      uVRegion {bIsValid=false, max=<0 0>, min=<0 0>}
     }
     pressedForeground {colorUseRule=UseColor_ColorTable, specifiedColor=<rgba 0 0 0 0>}
     pressedPadding=<ltrb 12 2.5 12 0.5>
     pressedSlateSound {resourceObject=null}
    }
   }
   slot {class=SizeBoxSlot}
  }
  TextBlock {
   widgetClassPath=&TextBlock
   parent=apply_control_rigs_btn
   widget=TextBlock
   properties {
    clipping=ClipToBounds
    text=Apply CtrlRigs
    font {
     bForceMonospaced=false
     bMaterialIsStencil=false
     fontMaterial=null
     fontObject=&Roboto
     letterSpacing=0
     monospacedWidth=1
     size=8.25
     skewAmount=0
     typefaceFontName=Bold
     outlineSettings {
      bApplyOutlineToDropShadows=true
      bMiteredCorners=false
      bSeparateFillAlpha=false
      outlineMaterial=null
      outlineSize=0
      outlineColor=<rgba 0 0 0 1>
     }
    }
   }
   slot {class=ButtonSlot, properties {padding=<ltrb 0 2 0 2>}}
  }
  SizeBox_3 {
   widgetClassPath=&SizeBox
   parent=VerticalBox_1
   widget=SizeBox_3
   properties {bOverride_MinDesiredWidth=true, minDesiredWidth=100}
   slot {class=VerticalBoxSlot, properties {horizontalAlignment=HAlign_Right, verticalAlignment=VAlign_Bottom, padding=<ltrb 0 3 0 0>}}
  }
  delete_control_rigs_btn {
   widgetClassPath=&EditorUtilityButton
   parent=SizeBox_3
   widget=delete_control_rigs_btn
   properties {
    clipping=ClipToBounds
    widgetStyle {
     clickedSlateSound {resourceObject=null}
     disabled {
      bIsDynamicallyLoaded=false
      drawAs=RoundedBox
      imageType=NoImage
      mirroring=NoMirror
      resourceName=None
      resourceObject=null
      tiling=NoTile
      imageSize=<0 0>
      margin=<ltrb 0 0 0 0>
      outlineSettings {
       bUseBrushTransparency=false
       roundingType=FixedRadius
       width=1
       color {colorUseRule=UseColor_ColorTable, specifiedColor=<rgba 0 0 0 0>}
       cornerRadii=<4 4 4 4>
      }
      tintColor {colorUseRule=UseColor_Specified, specifiedColor=<rgba 0.0395 0.0395 0.0395 1>}
      uVRegion {bIsValid=false, max=<0 0>, min=<0 0>}
     }
     disabledForeground {colorUseRule=UseColor_ColorTable, specifiedColor=<rgba 0 0 0 0>}
     hovered {
      bIsDynamicallyLoaded=false
      drawAs=RoundedBox
      imageType=NoImage
      mirroring=NoMirror
      resourceName=None
      resourceObject=null
      tiling=NoTile
      imageSize=<32 32>
      margin=<ltrb 0 0 0 0>
      outlineSettings {
       bUseBrushTransparency=false
       roundingType=FixedRadius
       width=1
       color {colorUseRule=UseColor_ColorTable, specifiedColor=<rgba 0 0 0 0>}
       cornerRadii=<4 4 4 4>
      }
      tintColor {colorUseRule=UseColor_Specified, specifiedColor=<rgba 0.4635 0.0579 0.0579 1>}
      uVRegion {bIsValid=false, max=<0 0>, min=<0 0>}
     }
     hoveredForeground {colorUseRule=UseColor_ColorTable, specifiedColor=<rgba 0 0 0 0>}
     hoveredSlateSound {resourceObject=null}
     normal {
      bIsDynamicallyLoaded=false
      drawAs=RoundedBox
      imageType=NoImage
      mirroring=NoMirror
      resourceName=None
      resourceObject=null
      tiling=NoTile
      imageSize=<32 32>
      margin=<ltrb 0 0 0 0>
      outlineSettings {
       bUseBrushTransparency=false
       roundingType=FixedRadius
       width=1
       color {colorUseRule=UseColor_ColorTable, specifiedColor=<rgba 0 0 0 0>}
       cornerRadii=<4 4 4 4>
      }
      tintColor {colorUseRule=UseColor_Specified, specifiedColor=<rgba 0.4635 0.1207 0.1207 1>}
      uVRegion {bIsValid=false, max=<0 0>, min=<0 0>}
     }
     normalForeground {colorUseRule=UseColor_ColorTable, specifiedColor=<rgba 0 0 0 0>}
     normalPadding=<ltrb 12 1.5 12 1.5>
     pressed {
      bIsDynamicallyLoaded=false
      drawAs=RoundedBox
      imageType=NoImage
      mirroring=NoMirror
      resourceName=None
      resourceObject=null
      tiling=NoTile
      imageSize=<32 32>
      margin=<ltrb 0 0 0 0>
      outlineSettings {
       bUseBrushTransparency=false
       roundingType=FixedRadius
       width=1
       color {colorUseRule=UseColor_ColorTable, specifiedColor=<rgba 0 0 0 0>}
       cornerRadii=<4 4 4 4>
      }
      tintColor {colorUseRule=UseColor_Specified, specifiedColor=<rgba 0.0284 0.0284 0.0284 1>}
      uVRegion {bIsValid=false, max=<0 0>, min=<0 0>}
     }
     pressedForeground {colorUseRule=UseColor_ColorTable, specifiedColor=<rgba 0 0 0 0>}
     pressedPadding=<ltrb 12 2.5 12 0.5>
     pressedSlateSound {resourceObject=null}
    }
   }
   slot {class=SizeBoxSlot}
  }
  TextBlock_63 {
   widgetClassPath=&TextBlock
   parent=delete_control_rigs_btn
   widget=TextBlock_63
   properties {
    clipping=ClipToBounds
    text=Delete CtrlRigs
    font {
     bForceMonospaced=false
     bMaterialIsStencil=false
     fontMaterial=null
     fontObject=&Roboto
     letterSpacing=0
     monospacedWidth=1
     size=8.25
     skewAmount=0
     typefaceFontName=Bold
     outlineSettings {
      bApplyOutlineToDropShadows=false
      bMiteredCorners=false
      bSeparateFillAlpha=false
      outlineMaterial=null
      outlineSize=0
      outlineColor=<rgba 0 0 0 1>
     }
    }
   }
   slot {class=ButtonSlot, properties {padding=<ltrb 0 2 0 2>}}
  }
  SizeBox {
   widgetClassPath=&SizeBox
   parent=VerticalBox_1
   widget=SizeBox
   properties {bOverride_MinDesiredWidth=true, minDesiredWidth=100}
   slot {class=VerticalBoxSlot, properties {horizontalAlignment=HAlign_Right, verticalAlignment=VAlign_Bottom, padding=<ltrb 0 3 0 0>}}
  }
  remove_twist_keys_btn {
   widgetClassPath=&EditorUtilityButton
   parent=SizeBox
   widget=remove_twist_keys_btn
   properties {
    clipping=ClipToBounds
    widgetStyle {
     clickedSlateSound {resourceObject=null}
     disabled {
      bIsDynamicallyLoaded=false
      drawAs=RoundedBox
      imageType=NoImage
      mirroring=NoMirror
      resourceName=None
      resourceObject=null
      tiling=NoTile
      imageSize=<0 0>
      margin=<ltrb 0 0 0 0>
      outlineSettings {
       bUseBrushTransparency=false
       roundingType=FixedRadius
       width=1
       color {colorUseRule=UseColor_ColorTable, specifiedColor=<rgba 0 0 0 0>}
       cornerRadii=<4 4 4 4>
      }
      tintColor {colorUseRule=UseColor_Specified, specifiedColor=<rgba 0.0395 0.0395 0.0395 1>}
      uVRegion {bIsValid=false, max=<0 0>, min=<0 0>}
     }
     disabledForeground {colorUseRule=UseColor_ColorTable, specifiedColor=<rgba 0 0 0 0>}
     hovered {
      bIsDynamicallyLoaded=false
      drawAs=RoundedBox
      imageType=NoImage
      mirroring=NoMirror
      resourceName=None
      resourceObject=null
      tiling=NoTile
      imageSize=<32 32>
      margin=<ltrb 0 0 0 0>
      outlineSettings {
       bUseBrushTransparency=false
       roundingType=FixedRadius
       width=1
       color {colorUseRule=UseColor_ColorTable, specifiedColor=<rgba 0 0 0 0>}
       cornerRadii=<4 4 4 4>
      }
      tintColor {colorUseRule=UseColor_Specified, specifiedColor=<rgba 0.7969 0.7257 0 1>}
      uVRegion {bIsValid=false, max=<0 0>, min=<0 0>}
     }
     hoveredForeground {colorUseRule=UseColor_ColorTable, specifiedColor=<rgba 0 0 0 0>}
     hoveredSlateSound {resourceObject=null}
     normal {
      bIsDynamicallyLoaded=false
      drawAs=RoundedBox
      imageType=NoImage
      mirroring=NoMirror
      resourceName=None
      resourceObject=null
      tiling=NoTile
      imageSize=<32 32>
      margin=<ltrb 0 0 0 0>
      outlineSettings {
       bUseBrushTransparency=false
       roundingType=FixedRadius
       width=1
       color {colorUseRule=UseColor_ColorTable, specifiedColor=<rgba 0 0 0 0>}
       cornerRadii=<4 4 4 4>
      }
      tintColor {colorUseRule=UseColor_Specified, specifiedColor=<rgba 0.4635 0.3992 0.0623 1>}
      uVRegion {bIsValid=false, max=<0 0>, min=<0 0>}
     }
     normalForeground {colorUseRule=UseColor_ColorTable, specifiedColor=<rgba 0 0 0 0>}
     normalPadding=<ltrb 12 1.5 12 1.5>
     pressed {
      bIsDynamicallyLoaded=false
      drawAs=RoundedBox
      imageType=NoImage
      mirroring=NoMirror
      resourceName=None
      resourceObject=null
      tiling=NoTile
      imageSize=<32 32>
      margin=<ltrb 0 0 0 0>
      outlineSettings {
       bUseBrushTransparency=false
       roundingType=FixedRadius
       width=1
       color {colorUseRule=UseColor_ColorTable, specifiedColor=<rgba 0 0 0 0>}
       cornerRadii=<4 4 4 4>
      }
      tintColor {colorUseRule=UseColor_Specified, specifiedColor=<rgba 0.0284 0.0284 0.0284 1>}
      uVRegion {bIsValid=false, max=<0 0>, min=<0 0>}
     }
     pressedForeground {colorUseRule=UseColor_ColorTable, specifiedColor=<rgba 0 0 0 0>}
     pressedPadding=<ltrb 12 2.5 12 0.5>
     pressedSlateSound {resourceObject=null}
    }
   }
   slot {class=SizeBoxSlot, properties {padding=<ltrb 0 0 0 5>}}
  }
  TextBlock_1 {
   widgetClassPath=&TextBlock
   parent=remove_twist_keys_btn
   widget=TextBlock_1
   properties {
    clipping=ClipToBounds
    text=Remove Twist Keys
    font {
     bForceMonospaced=false
     bMaterialIsStencil=false
     fontMaterial=null
     fontObject=&Roboto
     letterSpacing=0
     monospacedWidth=1
     size=8.25
     skewAmount=0
     typefaceFontName=Bold
     outlineSettings {
      bApplyOutlineToDropShadows=false
      bMiteredCorners=false
      bSeparateFillAlpha=false
      outlineMaterial=null
      outlineSize=0
      outlineColor=<rgba 0 0 0 1>
     }
    }
   }
   slot {class=ButtonSlot, properties {padding=<ltrb 0 2 0 2>}}
  }
  Spacer {
   widgetClassPath=&Spacer
   parent=HorizontalBox_0
   widget=Spacer
   properties {size=<10 1>}
   slot {class=HorizontalBoxSlot, properties {horizontalAlignment=HAlign_Left}}
  }
  SizeBox_0 {
   widgetClassPath=&SizeBox
   parent=HorizontalBox_0
   widget=SizeBox_0
   properties {minDesiredWidth=400}
   slot {class=HorizontalBoxSlot, properties {verticalAlignment=VAlign_Center, size {sizeRule=Fill, value=1}}}
  }
  Mannequin_Details {
   widgetClassPath=&DetailsView
   parent=SizeBox_0
   widget=Mannequin_Details
   properties {columnWidth=1, categoriesToShow [Character]}
   slot {class=SizeBoxSlot}
  }
  VerticalBox_2 {
   widgetClassPath=&VerticalBox
   parent=HorizontalBox_0
   widget=VerticalBox_2
   slot {class=HorizontalBoxSlot, properties {horizontalAlignment=HAlign_Center, verticalAlignment=VAlign_Center, padding=<ltrb 5 0 5 0>}}
  }
  configure_btn {
   widgetClassPath=&EditorUtilityButton
   parent=VerticalBox_2
   widget=configure_btn
   properties {widgetStyle {
     clickedSlateSound {resourceObject=null}
     disabled {
      bIsDynamicallyLoaded=false
      drawAs=RoundedBox
      imageType=NoImage
      mirroring=NoMirror
      resourceName=None
      resourceObject=null
      tiling=NoTile
      imageSize=<0 0>
      margin=<ltrb 0 0 0 0>
      outlineSettings {
       bUseBrushTransparency=false
       roundingType=FixedRadius
       width=1
       color {colorUseRule=UseColor_ColorTable, specifiedColor=<rgba 0 0 0 0>}
       cornerRadii=<4 4 4 4>
      }
      tintColor {colorUseRule=UseColor_Specified, specifiedColor=<rgba 0.0395 0.0395 0.0395 1>}
      uVRegion {bIsValid=false, max=<0 0>, min=<0 0>}
     }
     disabledForeground {colorUseRule=UseColor_ColorTable, specifiedColor=<rgba 0 0 0 0>}
     hovered {
      bIsDynamicallyLoaded=false
      drawAs=RoundedBox
      imageType=NoImage
      mirroring=NoMirror
      resourceName=None
      resourceObject=null
      tiling=NoTile
      imageSize=<32 32>
      margin=<ltrb 0 0 0 0>
      outlineSettings {
       bUseBrushTransparency=false
       roundingType=FixedRadius
       width=1
       color {colorUseRule=UseColor_ColorTable, specifiedColor=<rgba 0 0 0 0>}
       cornerRadii=<4 4 4 4>
      }
      tintColor {colorUseRule=UseColor_Specified, specifiedColor=<rgba 0.0953 0.0953 0.0953 1>}
      uVRegion {bIsValid=false, max=<0 0>, min=<0 0>}
     }
     hoveredForeground {colorUseRule=UseColor_ColorTable, specifiedColor=<rgba 0 0 0 0>}
     hoveredSlateSound {resourceObject=null}
     normal {
      bIsDynamicallyLoaded=false
      drawAs=RoundedBox
      imageType=NoImage
      mirroring=NoMirror
      resourceName=None
      resourceObject=null
      tiling=NoTile
      imageSize=<32 32>
      margin=<ltrb 0 0 0 0>
      outlineSettings {
       bUseBrushTransparency=false
       roundingType=FixedRadius
       width=1
       color {colorUseRule=UseColor_ColorTable, specifiedColor=<rgba 0 0 0 0>}
       cornerRadii=<4 4 4 4>
      }
      tintColor {colorUseRule=UseColor_Specified, specifiedColor=<rgba 0.0395 0.0395 0.0395 0>}
      uVRegion {bIsValid=false, max=<0 0>, min=<0 0>}
     }
     normalForeground {colorUseRule=UseColor_ColorTable, specifiedColor=<rgba 0 0 0 0>}
     normalPadding=<ltrb 0 0 0 0>
     pressed {
      bIsDynamicallyLoaded=false
      drawAs=RoundedBox
      imageType=NoImage
      mirroring=NoMirror
      resourceName=None
      resourceObject=null
      tiling=NoTile
      imageSize=<32 32>
      margin=<ltrb 0 0 0 0>
      outlineSettings {
       bUseBrushTransparency=false
       roundingType=FixedRadius
       width=1
       color {colorUseRule=UseColor_ColorTable, specifiedColor=<rgba 0 0 0 0>}
       cornerRadii=<4 4 4 4>
      }
      tintColor {colorUseRule=UseColor_Specified, specifiedColor=<rgba 0.0284 0.0284 0.0284 1>}
      uVRegion {bIsValid=false, max=<0 0>, min=<0 0>}
     }
     pressedForeground {colorUseRule=UseColor_ColorTable, specifiedColor=<rgba 0 0 0 0>}
     pressedPadding=<ltrb 0 0 0 0>
     pressedSlateSound {resourceObject=null}
    }}
   slot {class=VerticalBoxSlot}
  }
  Image_107 {
   widgetClassPath=&Image
   parent=configure_btn
   widget=Image_107
   properties {
    toolTipText=Configure mannequin settings
    brush {
     bIsDynamicallyLoaded=false
     drawAs=Image
     imageType=FullColor
     mirroring=NoMirror
     resourceName=None
     resourceObject=/Game/Athena/UI/InGame/Creative/PropertyEditor/Icon_Picker/Textures_V2/64x64/T_UI_IconLibrary_Gear_64
     tiling=NoTile
     imageSize=<18 18>
     margin=<ltrb 0 0 0 0>
     outlineSettings {
      bUseBrushTransparency=false
      roundingType=HalfHeightRadius
      width=0
      color {colorUseRule=UseColor_Specified, specifiedColor=<rgba 0 0 0 0>}
      cornerRadii=<0 0 0 0>
     }
     tintColor {colorUseRule=UseColor_Specified, specifiedColor=<rgba 1 1 1 1>}
     uVRegion {bIsValid=false, max=<0 0>, min=<0 0>}
    }
   }
   slot {class=ButtonSlot}
  }
 ]
}
