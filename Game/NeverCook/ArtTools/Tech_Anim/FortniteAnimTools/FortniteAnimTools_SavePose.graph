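# Editor utility dialog in the Fortnite Anim Tools for saving a Control Rig pose as an asset in a pose library.
# On construct it titles itself from the control rig's display name and binds a Details View to itself, showing only the Pose Asset Config category.
# Save broadcasts the SavePose dispatcher with directory, asset name, rig and an inverted selected-only flag, then closes; Cancel just closes.
# The actual saving is left to whatever binds that dispatcher.
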EditorUtilityWidgetBlueprint /Game/NeverCook/ArtTools/Tech_Anim/FortniteAnimTools/FortniteAnimTools_SavePose
graphs [
 EventGraph {nodes [
   n1 = "On Clicked (cancel_btn)"() {is=title K2Node_ComponentBoundEvent}
   RemoveFromParent(execute=[@n2.then, @n1.then])
   "On Clicked (save_pose_btn)"() {is=title K2Node_ComponentBoundEvent}
   > n2 = Call Save Pose!(
    execute:Exec
    self:Fortnite Anim Tools Save Pose Object Reference
    pose_library_directory:Directory Path Structure=Get pose_library_directory() -> (pose_library_directory:Directory Path Structure)
    pose_asset_name:String=Get pose_asset_name() -> (pose_asset_name:String)
    control_rig:Soft Object Reference=Get control_rig() -> (control_rig:Soft Object Reference)
    use_all_controls:bool=NOT Boolean(A=Get save_only_selected_controls() -> (save_only_selected_controls:bool))
   ) -> (then:Exec) {is=title K2Node_CallDelegate}
   Event Construct()
   n3 = GetDisplayName(Object=Resolve Soft Reference!(Input:Soft Object Reference=Get control_rig()) -> (Output:Object Reference))
   > Set Text(
    Text="To Text (String)"(InString=Append(A=@n3, B=" Pose"))
    self=Get control_rig_name_label() -> (control_rig_name_label:Text Object Reference)
   )
   > SetObject(self=Get pose_asset_config_details() -> (pose_asset_config_details:Details View Object Reference), NewObject=Self-Reference())
   Event PreConstruct()
   Event Tick()
  ]}
 SavePose {nodes [SavePose!() -> (then, pose_library_directory:Directory Path Structure, pose_asset_name:String, control_rig:Soft Object Reference, use_all_controls:bool=~) {is=title K2Node_FunctionEntry}]}
]
variables [
 control_rig {type=Object, default=null}
 pose_asset_name {type=string, default=""}
 pose_library_directory {type=DirectoryPath, default {path=""}}
 save_only_selected_controls {type=boolean}
 SavePose {default=()}
]
widgets {
 info {rootWidgetClass=&Border, widgetCount=11}
 widgets [
  Border_0 {widgetClassPath=&Border, widget=Border_0, properties {brushColor=<rgba 0.0262 0.0262 0.0262 1>}}
  VerticalBox_0 {widgetClassPath=&VerticalBox, parent=Border_0, widget=VerticalBox_0, slot {class=BorderSlot}}
  control_rig_name_label {
   widgetClassPath=&TextBlock
   bIsVariable=true
   parent=VerticalBox_0
   widget=control_rig_name_label
   properties {
    text=Text Block
    font {
     bForceMonospaced=false
     bMaterialIsStencil=false
     fontMaterial=null
     fontObject=&Roboto
     letterSpacing=0
     monospacedWidth=1
     size=12
     skewAmount=0
     typefaceFontName=Bold
     outlineSettings {
      bApplyOutlineToDropShadows=false
      bMiteredCorners=false
      bSeparateFillAlpha=false
      outlineMaterial=null
      outlineSize=0
      outlineColor=<rgba 0 0 0 1>
     }
    }
   }
   slot {class=VerticalBoxSlot, properties {horizontalAlignment=HAlign_Center, padding=<ltrb 0 5 0 0>}}
  }
  pose_asset_config_details {
   widgetClassPath=&DetailsView
   parent=VerticalBox_0
   widget=pose_asset_config_details
   properties {categoriesToShow [Pose Asset Config]}
   slot {class=VerticalBoxSlot}
  }
  Spacer_101 {
   widgetClassPath=&Spacer
   parent=VerticalBox_0
   widget=Spacer_101
   slot {class=VerticalBoxSlot, properties {size {sizeRule=Fill, value=1}}}
  }
  HorizontalBox_28 {widgetClassPath=&HorizontalBox, parent=VerticalBox_0, widget=HorizontalBox_28, slot {class=VerticalBoxSlot}}
  Spacer_167 {
   widgetClassPath=&Spacer
   parent=HorizontalBox_28
   widget=Spacer_167
   slot {class=HorizontalBoxSlot, properties {size {sizeRule=Fill, value=1}}}
  }
  save_pose_btn {
   widgetClassPath=&EditorUtilityButton
   parent=HorizontalBox_28
   widget=save_pose_btn
   slot {class=HorizontalBoxSlot, properties {size {sizeRule=Automatic, value=0.2}}}
  }
  TextBlock_118 {
   widgetClassPath=&TextBlock
   parent=save_pose_btn
   widget=TextBlock_118
   properties {
    text=Save Pose
    font {
     bForceMonospaced=false
     bMaterialIsStencil=false
     fontMaterial=null
     fontObject=&Roboto
     letterSpacing=0
     monospacedWidth=1
     size=9
     skewAmount=0
     typefaceFontName=Bold
     outlineSettings {
      bApplyOutlineToDropShadows=false
      bMiteredCorners=false
      bSeparateFillAlpha=false
      outlineMaterial=null
      outlineSize=0
      outlineColor=<rgba 0 0 0 1>
     }
    }
   }
   slot {class=ButtonSlot}
  }
  cancel_btn {
   widgetClassPath=&EditorUtilityButton
   parent=HorizontalBox_28
   widget=cancel_btn
   slot {class=HorizontalBoxSlot, properties {size {sizeRule=Automatic, value=0.2}}}
  }
  TextBlock_59 {
   widgetClassPath=&TextBlock
   parent=cancel_btn
   widget=TextBlock_59
   properties {
    text=Cancel
    font {
     bForceMonospaced=false
     bMaterialIsStencil=false
     fontMaterial=null
     fontObject=&Roboto
     letterSpacing=0
     monospacedWidth=1
     size=9
     skewAmount=0
     typefaceFontName=Bold
     outlineSettings {
      bApplyOutlineToDropShadows=false
      bMiteredCorners=false
      bSeparateFillAlpha=false
      outlineMaterial=null
      outlineSize=0
      outlineColor=<rgba 0 0 0 1>
     }
    }
   }
   slot {class=ButtonSlot}
  }
 ]
}
